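# Physics asset for the Cape FaceAcc mesh part of the female CasinoReaper Die skin, covering two dynamic hair-tail chains and the body.
# Left and right ten-bone capsule chains hang from the head, linked by constraints with tight swing and very narrow twist limits.
# Capsules on spine, limbs and head form the body collision, joined by free-rotation constraints.
# An oversized offset pelvis capsule and a huge sphere below the root add broad collision volumes.
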PhysicsAsset /BRCosmetics/Characters/Player/Female/Medium/Bodies/F_MED_CasinoReaper_Die/Meshes/Parts/F_MED_CasinoReaper_Die_Cape_FaceAcc_Physics
bodies {
 calf_l [{shapeType=Capsule, length=30.5, radius=7.505, shapeName=calf_l_capsule, center=<-22.9542 0 0>, rotation=<rot 0 -90 90>}]
 calf_r [{shapeType=Capsule, length=30.5, radius=7.505, shapeName=calf_r_capsule, center=<23.4421 0 0>, rotation=<rot 0 -90 90>}]
 clavicle_l [{shapeType=Capsule, length=7, radius=5.505, shapeName=clavicle_l_capsule, center=<5.2222 0 0>, rotation=<rot 0 -90 90>}]
 clavicle_r [{shapeType=Capsule, length=7.5, radius=5.505, shapeName=clavicle_r_capsule, center=<-4.8195 0 0>, rotation=<rot 0 -90 90>}]
 dyn_hair_tail_01_l [{
   shapeType=Capsule
   length=3.7869
   radius=3.7464
   shapeName=dyn_hair_tail_base_l_capsule
   center=<5.8092 0.4807 0.166>
   rotation=<rot 0 90 90>
  }]
 dyn_hair_tail_01_r [{
   shapeType=Capsule
   length=3.7869
   radius=3.7464
   shapeName=dyn_hair_tail_base_l_capsule
   center=<5.8092 0.4807 0.166>
   rotation=<rot 0 90 90>
  }]
 dyn_hair_tail_02_l [{shapeType=Capsule, length=9.5052, radius=2.2426, shapeName=dyn_hair_tail_02_l_capsule, center=<6.1078 -0.123 0.1567>, rotation=<rot 0 90 90>}]
 dyn_hair_tail_02_r [{
   shapeType=Capsule
   length=9.6294
   radius=2.4401
   shapeName=dyn_hair_tail_02_r_capsule
   center=<4.9296 -0.2606 -0.1082>
   rotation=<rot 0 90 90>
  }]
 dyn_hair_tail_03_l [{
   shapeType=Capsule
   length=6.818
   radius=2.2846
   shapeName=dyn_hair_tail_03_l_capsule
   center=<8.5689 -0.0236 -0.0147>
   rotation=<rot 0 90 90>
  }]
 dyn_hair_tail_03_r [{shapeType=Capsule, length=6.8, radius=2.318, shapeName=dyn_hair_tail_03_r_capsule, center=<7.8443 -0.0006 -0.0989>, rotation=<rot 0 90 90>}]
 dyn_hair_tail_04_l [{
   shapeType=Capsule
   length=6.9305
   radius=2.0077
   shapeName=dyn_hair_tail_04_l_capsule
   center=<8.2128 -0.0544 -0.2015>
   rotation=<rot 0 90 90>
  }]
 dyn_hair_tail_04_r [{
   shapeType=Capsule
   length=7.4953
   radius=1.9945
   shapeName=dyn_hair_tail_04_r_capsule
   center=<8.1094 -0.0503 0.1255>
   rotation=<rot 0 90 90>
  }]
 dyn_hair_tail_05_l [{
   shapeType=Capsule
   length=7.7751
   radius=1.6565
   shapeName=dyn_hair_tail_05_l_capsule
   center=<8.7449 -0.1249 -0.2057>
   rotation=<rot 0 90 90>
  }]
 dyn_hair_tail_05_r [{
   shapeType=Capsule
   length=7.5633
   radius=1.6429
   shapeName=dyn_hair_tail_05_r_capsule
   center=<9.1181 -0.1083 0.1653>
   rotation=<rot 0 90 90>
  }]
 dyn_hair_tail_06_l [{
   shapeType=Capsule
   length=6.9416
   radius=1.3359
   shapeName=dyn_hair_tail_06_l_capsule
   center=<8.6404 0.0759 -0.0057>
   rotation=<rot 0 90 90>
  }]
 dyn_hair_tail_06_r [{shapeType=Capsule, length=6.3434, radius=1.2884, shapeName=dyn_hair_tail_06_r_capsule, center=<8.3802 0.0658 0.0328>, rotation=<rot 0 90 90>}]
 dyn_hair_tail_07_l [{
   shapeType=Capsule
   length=6.1201
   radius=1.5591
   shapeName=dyn_hair_tail_07_l_capsule
   center=<8.2834 0.0221 -0.0214>
   rotation=<rot 0 90 90>
  }]
 dyn_hair_tail_07_r [{
   shapeType=Capsule
   length=7.1012
   radius=1.6045
   shapeName=dyn_hair_tail_07_r_capsule
   center=<8.2472 -0.0016 -0.0289>
   rotation=<rot 0 90 90>
  }]
 dyn_hair_tail_08_l [{shapeType=Capsule, length=7.3992, radius=1.2614, shapeName=dyn_hair_tail_08_l_capsule, center=<8.3691 -0.018 -0.102>, rotation=<rot 0 90 90>}]
 dyn_hair_tail_08_r [{
   shapeType=Capsule
   length=5.9533
   radius=1.2946
   shapeName=dyn_hair_tail_08_r_capsule
   center=<8.4522 -0.0175 0.0282>
   rotation=<rot 0 90 90>
  }]
 dyn_hair_tail_09_l [{
   shapeType=Capsule
   length=12.6825
   radius=1.1781
   shapeName=dyn_hair_tail_09_l_capsule
   center=<9.0311 -0.235 -0.0044>
   rotation=<rot 0 90 90>
  }]
 dyn_hair_tail_09_r [{
   shapeType=Capsule
   length=12.8249
   radius=1.1918
   shapeName=dyn_hair_tail_09_r_capsule
   center=<8.3316 -0.1294 0.0176>
   rotation=<rot 0 90 90>
  }]
 dyn_hair_tail_base_l [{shapeType=Capsule, length=1.0333, radius=0.5, shapeName=dyn_hair_tail_base_l_capsule, rotation=<rot 0 90 90>}]
 dyn_hair_tail_base_r [{shapeType=Capsule, length=0.5, radius=0.5, shapeName=dyn_hair_tail_base_r_capsule, rotation=<rot 0 90 90>}]
 foot_l [{shapeType=Capsule, length=5, radius=2.505, shapeName=foot_l_capsule, rotation=<rot 0 -90 90>}]
 foot_r [{shapeType=Capsule, length=5, radius=2.505, shapeName=foot_r_capsule, rotation=<rot 0 -90 90>}]
 head [{shapeType=Capsule, length=4.4639, radius=2.981, shapeName=head_capsule, center=<4.0188 -0.1979 -0.175>, rotation=<rot 0 90 90>}]
 lowerarm_l [{shapeType=Capsule, length=27.5, radius=4.505, shapeName=lowerarm_l_capsule, center=<15.2839 0 0>, rotation=<rot 0 -90 90>}]
 lowerarm_r [{shapeType=Capsule, length=27.5, radius=4.505, shapeName=lowerarm_r_capsule, center=<-13.5551 0 0>, rotation=<rot 0 -90 90>}]
 pelvis [
  {shapeType=Capsule, length=13.005, radius=12.505, shapeName=pelvis_capsule, center=<-3.4775 0 0>, rotation=<rot 0 -90 0>}
  {shapeType=Capsule, length=15, radius=100, shapeName=pelvis_capsule, center=<-20.4895 -117.2796 0>}
 ]
 root [{shapeType=Sphere, radius=300, shapeName=root_sphere, center=<0 0 -291.0632>}]
 spine_01 [{shapeType=Capsule, length=8.5, radius=11.005, shapeName=spine_01_capsule, rotation=<rot 0 -90 0>}]
 spine_03 [{shapeType=Capsule, length=10, radius=9.505, shapeName=spine_03_capsule, rotation=<rot 0 -90 0>}]
 spine_05 [{shapeType=Capsule, length=10.5, radius=9.005, shapeName=spine_05_capsule, rotation=<rot 0 -90 0>}]
 thigh_l [{shapeType=Capsule, length=30.5, radius=7.505, shapeName=thigh_l_capsule, center=<-21.0697 0 0>, rotation=<rot 0 -90 90>}]
 thigh_r [{shapeType=Capsule, length=30.5, radius=7.505, shapeName=thigh_r_capsule, center=<20.4349 0 0>, rotation=<rot 0 -90 90>}]
 upperarm_l [{shapeType=Capsule, length=27.5, radius=4.505, shapeName=upperarm_l_capsule, center=<15.2839 0 0>, rotation=<rot 0 -90 90>}]
 upperarm_r [{shapeType=Capsule, length=27.5, radius=4.505, shapeName=upperarm_r_capsule, center=<-13.5551 0 0>, rotation=<rot 0 -90 90>}]
}
constraints [
 {bone1Name=dyn_hair_tail_01_l, bone2Name=dyn_hair_tail_base_l, swing1LimitDegrees=10, swing2LimitDegrees=10, twistLimitDegrees=2}
 {bone1Name=dyn_hair_tail_02_l, bone2Name=dyn_hair_tail_01_l, swing1LimitDegrees=10, swing2LimitDegrees=10, twistLimitDegrees=2}
 {bone1Name=dyn_hair_tail_03_l, bone2Name=dyn_hair_tail_02_l, swing1LimitDegrees=10, swing2LimitDegrees=10, twistLimitDegrees=2}
 {bone1Name=dyn_hair_tail_04_l, bone2Name=dyn_hair_tail_03_l, swing1LimitDegrees=15, swing2LimitDegrees=10, twistLimitDegrees=2}
 {bone1Name=dyn_hair_tail_05_l, bone2Name=dyn_hair_tail_04_l, swing1LimitDegrees=15, swing2LimitDegrees=10, twistLimitDegrees=2}
 {bone1Name=dyn_hair_tail_06_l, bone2Name=dyn_hair_tail_05_l, swing1LimitDegrees=10, swing2LimitDegrees=10, twistLimitDegrees=2}
 {bone1Name=dyn_hair_tail_07_l, bone2Name=dyn_hair_tail_06_l, swing1LimitDegrees=10, swing2LimitDegrees=10, twistLimitDegrees=2}
 {bone1Name=dyn_hair_tail_08_l, bone2Name=dyn_hair_tail_07_l, swing1LimitDegrees=10, swing2LimitDegrees=10, twistLimitDegrees=2}
 {bone1Name=dyn_hair_tail_09_l, bone2Name=dyn_hair_tail_08_l, swing1LimitDegrees=10, swing2LimitDegrees=10, twistLimitDegrees=2}
 {bone1Name=dyn_hair_tail_09_r, bone2Name=dyn_hair_tail_08_r, swing1LimitDegrees=10, swing2LimitDegrees=10, twistLimitDegrees=2}
 {bone1Name=dyn_hair_tail_01_r, bone2Name=dyn_hair_tail_base_r, swing1LimitDegrees=10, swing2LimitDegrees=10, twistLimitDegrees=2}
 {bone1Name=dyn_hair_tail_02_r, bone2Name=dyn_hair_tail_01_r, swing1LimitDegrees=10, swing2LimitDegrees=10, twistLimitDegrees=2}
 {bone1Name=dyn_hair_tail_03_r, bone2Name=dyn_hair_tail_02_r, swing1LimitDegrees=10, swing2LimitDegrees=10, twistLimitDegrees=2}
 {bone1Name=dyn_hair_tail_04_r, bone2Name=dyn_hair_tail_03_r, swing1LimitDegrees=15, swing2LimitDegrees=10, twistLimitDegrees=2}
 {bone1Name=dyn_hair_tail_05_r, bone2Name=dyn_hair_tail_04_r, swing1LimitDegrees=15, swing2LimitDegrees=10, twistLimitDegrees=2}
 {bone1Name=dyn_hair_tail_06_r, bone2Name=dyn_hair_tail_05_r, swing1LimitDegrees=10, swing2LimitDegrees=10, twistLimitDegrees=2}
 {bone1Name=dyn_hair_tail_07_r, bone2Name=dyn_hair_tail_06_r, swing1LimitDegrees=10, swing2LimitDegrees=10, twistLimitDegrees=2}
 {bone1Name=dyn_hair_tail_08_r, bone2Name=dyn_hair_tail_07_r, swing1LimitDegrees=10, swing2LimitDegrees=10, twistLimitDegrees=2}
 {bone1Name=dyn_hair_tail_base_r, bone2Name=head, swing1LimitDegrees=20, swing2LimitDegrees=10, twistLimitDegrees=2}
 {bone1Name=dyn_hair_tail_base_l, bone2Name=head, swing1LimitDegrees=20, swing2LimitDegrees=20, twistLimitDegrees=2}
 {
  bone1Name=head
  bone2Name=spine_05
  swing1LimitDegrees=45
  swing1Motion=Free
  swing2LimitDegrees=45
  swing2Motion=Free
  twistLimitDegrees=45
  twistMotion=Free
 }
 {
  bone1Name=spine_05
  bone2Name=spine_03
  swing1LimitDegrees=45
  swing1Motion=Free
  swing2LimitDegrees=45
  swing2Motion=Free
  twistLimitDegrees=45
  twistMotion=Free
 }
 {
  bone1Name=spine_03
  bone2Name=spine_01
  swing1LimitDegrees=45
  swing1Motion=Free
  swing2LimitDegrees=45
  swing2Motion=Free
  twistLimitDegrees=45
  twistMotion=Free
 }
 {
  bone1Name=spine_01
  bone2Name=pelvis
  swing1LimitDegrees=45
  swing1Motion=Free
  swing2LimitDegrees=45
  swing2Motion=Free
  twistLimitDegrees=45
  twistMotion=Free
 }
 {
  bone1Name=thigh_l
  bone2Name=pelvis
  swing1LimitDegrees=45
  swing1Motion=Free
  swing2LimitDegrees=45
  swing2Motion=Free
  twistLimitDegrees=45
  twistMotion=Free
 }
 {
  bone1Name=calf_l
  bone2Name=thigh_l
  swing1LimitDegrees=45
  swing1Motion=Free
  swing2LimitDegrees=45
  swing2Motion=Free
  twistLimitDegrees=45
  twistMotion=Free
 }
 {
  bone1Name=foot_l
  bone2Name=calf_l
  swing1LimitDegrees=45
  swing1Motion=Free
  swing2LimitDegrees=45
  swing2Motion=Free
  twistLimitDegrees=45
  twistMotion=Free
 }
 {
  bone1Name=thigh_r
  bone2Name=pelvis
  swing1LimitDegrees=45
  swing1Motion=Free
  swing2LimitDegrees=45
  swing2Motion=Free
  twistLimitDegrees=45
  twistMotion=Free
 }
 {
  bone1Name=calf_r
  bone2Name=thigh_r
  swing1LimitDegrees=45
  swing1Motion=Free
  swing2LimitDegrees=45
  swing2Motion=Free
  twistLimitDegrees=45
  twistMotion=Free
 }
 {
  bone1Name=foot_r
  bone2Name=calf_r
  swing1LimitDegrees=45
  swing1Motion=Free
  swing2LimitDegrees=45
  swing2Motion=Free
  twistLimitDegrees=45
  twistMotion=Free
 }
 {
  bone1Name=lowerarm_r
  bone2Name=upperarm_r
  swing1LimitDegrees=45
  swing1Motion=Free
  swing2LimitDegrees=45
  swing2Motion=Free
  twistLimitDegrees=45
  twistMotion=Free
 }
 {
  bone1Name=clavicle_l
  bone2Name=spine_05
  swing1LimitDegrees=45
  swing1Motion=Free
  swing2LimitDegrees=45
  swing2Motion=Free
  twistLimitDegrees=45
  twistMotion=Free
 }
 {
  bone1Name=upperarm_l
  bone2Name=clavicle_l
  swing1LimitDegrees=45
  swing1Motion=Free
  swing2LimitDegrees=45
  swing2Motion=Free
  twistLimitDegrees=45
  twistMotion=Free
 }
 {
  bone1Name=lowerarm_l
  bone2Name=upperarm_l
  swing1LimitDegrees=45
  swing1Motion=Free
  swing2LimitDegrees=45
  swing2Motion=Free
  twistLimitDegrees=45
  twistMotion=Free
 }
 {
  bone1Name=clavicle_r
  bone2Name=spine_05
  swing1LimitDegrees=45
  swing1Motion=Free
  swing2LimitDegrees=45
  swing2Motion=Free
  twistLimitDegrees=45
  twistMotion=Free
 }
 {
  bone1Name=upperarm_r
  bone2Name=clavicle_r
  swing1LimitDegrees=45
  swing1Motion=Free
  swing2LimitDegrees=45
  swing2Motion=Free
  twistLimitDegrees=45
  twistMotion=Free
 }
 {
  bone1Name=pelvis
  bone2Name=root
  swing1LimitDegrees=45
  swing1Motion=Free
  swing2LimitDegrees=45
  swing2Motion=Free
  twistLimitDegrees=45
  twistMotion=Free
 }
]
