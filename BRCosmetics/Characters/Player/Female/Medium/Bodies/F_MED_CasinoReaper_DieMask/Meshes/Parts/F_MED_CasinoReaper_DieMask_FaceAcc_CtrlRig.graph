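# Control Rig for the CasinoReaper DieMask female face accessory that rotates the alt_hair_corrective bone from leg pose.
# Spherical pose readers on both thighs and calves feed greater-than selects, whose sum weights a Modify Transforms offset.
# The calf select takes the larger value; the thigh select outputs right calf, not right thigh, when right thigh wins.
# Begin Execution is not connected to this chain.
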ControlRigBlueprint /BRCosmetics/Characters/Player/Female/Medium/Bodies/F_MED_CasinoReaper_DieMask/Meshes/Parts/F_MED_CasinoReaper_DieMask_FaceAcc_CtrlRig
graphs [
 RigVMFunctionLibrary {links [], nodes []}
 RigVMModel {nodes [
   n1 = &RigUnit_SphericalPoseReader(
    ActiveRegionScaleFactors=(PositiveWidth=1,NegativeWidth=1,PositiveHeight=1,NegativeHeight=1)
    ActiveRegionSize=0.1
    Debug=(bDrawDebug=True,bDraw2D=False,bDrawLocalAxes=False,DebugScale=25,DebugSegments=20,DebugThickness=0.25)
    DriverAxis=(X=1,Y=0,Z=0)
    DriverItem=(Type=Bone,Name="thigh_r")
    ExecutePin=()
    FalloffRegionScaleFactors=(PositiveWidth=1,NegativeWidth=1,PositiveHeight=1,NegativeHeight=1)
    FalloffSize=0.45
    FlipHeightScaling=False
    FlipWidthScaling=False
    OptionalParentItem=(Type=Bone,Name="None")
    RotationOffset=(X=90,Y=0,Z=0)
   )
   > n2 = &RigUnit_SphericalPoseReader(
    ActiveRegionScaleFactors=(PositiveWidth=1,NegativeWidth=1,PositiveHeight=1,NegativeHeight=1)
    ActiveRegionSize=0.1
    Debug=(bDrawDebug=True,bDraw2D=False,bDrawLocalAxes=False,DebugScale=25,DebugSegments=20,DebugThickness=0.25)
    DriverAxis=(X=1,Y=0,Z=0)
    DriverItem=(Type=Bone,Name="calf_r")
    FalloffRegionScaleFactors=(PositiveWidth=1,NegativeWidth=1,PositiveHeight=1,NegativeHeight=1)
    FalloffSize=0.8
    FlipHeightScaling=False
    FlipWidthScaling=False
    OptionalParentItem=(Type=Bone,Name="None")
    RotationOffset=(X=90,Y=0,Z=90)
   )
   > n3 = &RigUnit_SphericalPoseReader(
    ActiveRegionScaleFactors=(PositiveWidth=1,NegativeWidth=1,PositiveHeight=1,NegativeHeight=1)
    ActiveRegionSize=0.1
    Debug=(bDrawDebug=True,bDraw2D=False,bDrawLocalAxes=False,DebugScale=25,DebugSegments=20,DebugThickness=0.25)
    DriverAxis=(X=-1,Y=0,Z=0)
    DriverItem=(Type=Bone,Name="thigh_l")
    FalloffRegionScaleFactors=(PositiveWidth=1,NegativeWidth=1,PositiveHeight=1,NegativeHeight=1)
    FalloffSize=0.45
    FlipHeightScaling=False
    FlipWidthScaling=False
    OptionalParentItem=(Type=Bone,Name="None")
    RotationOffset=(X=-90,Y=0,Z=0)
   )
   > n4 = &RigUnit_SphericalPoseReader(
    ActiveRegionScaleFactors=(PositiveWidth=1,NegativeWidth=1,PositiveHeight=1,NegativeHeight=1)
    ActiveRegionSize=0.1
    Debug=(bDrawDebug=True,bDraw2D=False,bDrawLocalAxes=False,DebugScale=25,DebugSegments=20,DebugThickness=0.25)
    DriverAxis=(X=-1,Y=0,Z=0)
    DriverItem=(Type=Bone,Name="calf_l")
    FalloffRegionScaleFactors=(PositiveWidth=1,NegativeWidth=1,PositiveHeight=1,NegativeHeight=1)
    FalloffSize=0.8
    FlipHeightScaling=False
    FlipWidthScaling=False
    OptionalParentItem=(Type=Bone,Name="None")
    RotationOffset=(X=-90,Y=0,Z=90)
   )
   Greater = /Script/RigVM.RigVMFunction_MathDoubleGreater(A=@n2.OutputParam, B=@n4.OutputParam) {is=struct RigVMUnitNode, template="Greater::Execute(in A,in B,out Result)"}
   Add = &RigVMFunction_MathDoubleAdd(
    B=RigVMDispatchNode(Condition=@Greater.Result, True=@n2.OutputParam, False=@n4.OutputParam) {template="DISPATCH_RigVMDispatch_If(in Condition,in True,in False,out Result)"}.Result
    A=RigVMDispatchNode(
     Condition=&RigVMFunction_MathDoubleGreater(A=@n1.OutputParam, B=@n3.OutputParam) {template="Greater::Execute(in A,in B,out Result)"}.Result
     True=@n2.OutputParam
     False=@n3.OutputParam
    ) {template="DISPATCH_RigVMDispatch_If(in Condition,in True,in False,out Result)"}.Result
   )
   > &RigUnit_ModifyTransforms(
    ItemToModify=((Item=(Type=Bone,Name="alt_hair_corrective"),Transform=(Rotation=(X=-0.4226,Y=0,Z=0,W=0.9063),Translation=(X=0,Y=0,Z=0),Scale3D=(X=1,Y=1,Z=1))))
    WeightMinimum=0
    Weight=@Add.Result
   )
   RigUnit_BeginExecution = &RigUnit_BeginExecution()
  ]}
]
hierarchy [
 root {loc=<0 0 0>, rot=<0 0 0 1>}
 pelvis {parent=root, loc=<0 -0.5301 98.0571>, rot=<0.0127 0.707 -0.0127 -0.707>}
 spine_01 {parent=pelvis, loc=<6.1208 -0.9328 0>, rot=<0 0 0.034 -0.9994>}
 spine_02 {parent=spine_01, loc=<6.5997 -0.2095 0>, rot=<0 0 0.0238 -0.9997>}
 spine_03 {parent=spine_02, loc=<6.4914 -0.1237 0>, rot=<0 0 0.0149 -0.9999>}
 spine_04 {parent=spine_03, loc=<6.9598 0.0794 0>, rot=<0 0 -0.0132 -0.9999>}
 spine_05 {parent=spine_04, loc=<6.9987 0 0>, rot=<0 0 -0.0015 -1>}
 attach_cape {parent=spine_05, loc=<9.3216 -8.4501 0>, rot=<0.0616 -0.7044 -0.0616 -0.7044>}
 neck_01 {parent=spine_05, loc=<11.3138 0.1529 0>, rot=<0 0 -0.2003 -0.9797>}
 neck_02 {parent=neck_01, loc=<4.5627 0 0>, rot=<0 0 0.0504 -0.9987>}
 head {parent=neck_02, loc=<5.3914 0 0>, rot=<0 0 0.115 -0.9934>}
 dyn_mask_01_r {parent=head, loc=<0.1691 9.6708 1.3497>, rot=<-0.0127 0.5543 0.8322 -0.0032>}
 dyn_mask_02_r {parent=head, loc=<1.4434 8.3412 3.9184>, rot=<-0.0127 0.5543 0.8322 -0.0032>}
 dyn_mask_03_r {parent=head, loc=<2.1611 6.0012 5.7649>, rot=<-0.0115 0.3239 0.946 -0.0064>}
 dyn_mask_01_l {parent=head, loc=<0.1711 9.4334 -1.4439>, rot=<0.0131 -0.7962 -0.6049 -0.0011>}
 dyn_mask_02_l {parent=head, loc=<1.4434 8.3412 -3.9697>, rot=<0.0126 -0.895 -0.4459 -0.0035>}
 dyn_mask_03_l {parent=head, loc=<2.1611 6.0012 -5.7905>, rot=<0.0119 -0.9565 -0.2915 -0.0056>}
 hair_base {parent=head, loc=<7.232 -5.4757 -0.2126>, rot=<0.0041 -1 0 -0.0083>}
 dyn_hair_tail_base_l {parent=hair_base, loc=<5.2547 -1.167 4.7764>, rot=<0.0011 -0.0083 0.1326 -0.9911>}
 dyn_hair_tail_01_l {parent=dyn_hair_tail_base_l, loc=<0 0 0>, rot=<0 0 0 1>}
 dyn_hair_tail_02_l {parent=dyn_hair_tail_01_l, loc=<15.9816 -0.0638 0.0703>, rot=<0 0 -0.0673 -0.9977>}
 dyn_hair_tail_03_l {parent=dyn_hair_tail_02_l, loc=<12.7418 0 0>, rot=<0 0 -0.0317 -0.9995>}
 dyn_hair_tail_04_l {parent=dyn_hair_tail_03_l, loc=<13.9113 0 0>, rot=<0 0 -0.011 -0.9999>}
 dyn_hair_tail_05_l {parent=dyn_hair_tail_04_l, loc=<13.9251 0 0>, rot=<0 0 -0.0075 -1>}
 dyn_hair_tail_06_l {parent=dyn_hair_tail_05_l, loc=<14.4993 0 0>, rot=<0 0 0.0075 -1>}
 dyn_hair_tail_07_l {parent=dyn_hair_tail_06_l, loc=<13.9251 0 0>, rot=<0 0 0 1>}
 dyn_hair_tail_08_l {parent=dyn_hair_tail_07_l, loc=<13.9251 0 0>, rot=<0 0 0.0193 -0.9998>}
 dyn_hair_tail_09_l {parent=dyn_hair_tail_08_l, loc=<14.0686 0 0>, rot=<0 0 0 1>}
 dyn_hair_tail_base_r {parent=hair_base, loc=<5.0854 -1.167 -5.3752>, rot=<0.0011 -0.0083 0.1326 -0.9911>}
 dyn_hair_tail_01_r {parent=dyn_hair_tail_base_r, loc=<0 0 0>, rot=<0 0 0 1>}
 dyn_hair_tail_02_r {parent=dyn_hair_tail_01_r, loc=<15.9816 -0.0638 0>, rot=<0 0 -0.0673 -0.9977>}
 dyn_hair_tail_03_r {parent=dyn_hair_tail_02_r, loc=<12.7418 0 0>, rot=<0 0 -0.0317 -0.9995>}
 dyn_hair_tail_04_r {parent=dyn_hair_tail_03_r, loc=<13.9113 0 0>, rot=<0 0 -0.011 -0.9999>}
 dyn_hair_tail_05_r {parent=dyn_hair_tail_04_r, loc=<13.9251 0 0>, rot=<0 0 -0.0075 -1>}
 dyn_hair_tail_06_r {parent=dyn_hair_tail_05_r, loc=<14.4993 0 0>, rot=<0 0 0.0075 -1>}
 dyn_hair_tail_07_r {parent=dyn_hair_tail_06_r, loc=<13.9251 0 0>, rot=<0 0 0 1>}
 dyn_hair_tail_08_r {parent=dyn_hair_tail_07_r, loc=<13.9251 0 0>, rot=<0 0 0.0193 -0.9998>}
 dyn_hair_tail_09_r {parent=dyn_hair_tail_08_r, loc=<14.0686 0 0>, rot=<0 0 0 1>}
 dyn_hat_b_base {parent=hair_base, loc=<2.3713 -0.6585 -0.2612>, rot=<-0.4957 0.4958 0.5044 -0.504>}
 dyn_hat_b {parent=dyn_hat_b_base, loc=<0 0 0>, rot=<0 0 0 1>}
 dyn_hat_r {parent=head, loc=<7.1204 0.7913 3.7269>, rot=<0.7071 0.0029 0.7071 -0.0029>}
 dyn_hat_f {parent=head, loc=<7.7241 6.5637 0.0058>, rot=<-0.4979 -0.5021 -0.5021 -0.4979>}
 dyn_hat_l {parent=head, loc=<7.1204 0.7913 -3.775>, rot=<0.0029 -0.7071 -0.0029 -0.7071>}
 dyn_hat_rf {parent=head, loc=<6.9633 5.2013 3.5594>, rot=<0.6846 0.1771 0.686 0.1715>}
 dyn_hat_lf {parent=head, loc=<6.9633 5.2013 -3.3589>, rot=<-0.2855 -0.6469 -0.2908 -0.6445>}
 dyn_hat_rb_base {parent=head, loc=<6.0946 -2.748 3.5594>, rot=<0.6845 -0.1774 0.683 -0.1831>}
 dyn_hat_rb {parent=dyn_hat_rb_base, loc=<0 0 0>, rot=<0 0 0 1>}
 dyn_hat_lb_base {parent=head, loc=<6.0946 -2.748 -3.3589>, rot=<0.2635 -0.6562 0.2581 -0.6583>}
 dyn_hat_lb {parent=dyn_hat_lb_base, loc=<0 0 0>, rot=<0 0 0 1>}
 FACIAL_C_FacialRoot {parent=head, loc=<2.1524 -0.746 -0.0002>, rot=<0.4879 -0.5118 -0.4879 -0.5118>}
 FACIAL_C_Forehead {parent=FACIAL_C_FacialRoot, loc=<-0.0016 -2.978 8.8371>, rot=<-0.0547 0 0 -0.9985>}
 FACIAL_L_ForeheadIn {parent=FACIAL_C_FacialRoot, loc=<1.6788 -3.0227 8.7129>, rot=<-0.072 -0.0696 0.0035 -0.995>}
 FACIAL_R_ForeheadIn {parent=FACIAL_C_FacialRoot, loc=<-1.681 -3.0201 8.7109>, rot=<-0.072 0.0697 -0.0035 -0.995>}
 FACIAL_L_ForeheadMid {parent=FACIAL_C_FacialRoot, loc=<3.2299 -3.2702 8.1585>, rot=<-0.0821 -0.1754 -0.0226 -0.9808>}
 FACIAL_R_ForeheadMid {parent=FACIAL_C_FacialRoot, loc=<-3.2295 -3.2673 8.1556>, rot=<-0.0821 0.1754 0.0226 -0.9808>}
 FACIAL_L_ForeheadOut {parent=FACIAL_C_FacialRoot, loc=<4.6192 -3.2158 7.1127>, rot=<-0.1322 -0.2547 -0.0821 -0.9544>}
 FACIAL_R_ForeheadOut {parent=FACIAL_C_FacialRoot, loc=<-4.6183 -3.213 7.1109>, rot=<-0.1322 0.2547 0.082 -0.9544>}
 FACIAL_L_EyesackUpper {parent=FACIAL_C_FacialRoot, loc=<3.3792 -2.0563 8.8884>, rot=<0.0376 -0.0835 -0.0655 -0.9936>}
 FACIAL_R_EyesackUpper {parent=FACIAL_C_FacialRoot, loc=<-3.3807 -2.0563 8.8876>, rot=<0.0376 0.0835 0.0655 -0.9936>}
 FACIAL_L_EyelidUpperFurrow {parent=FACIAL_C_FacialRoot, loc=<3.2051 -1.7948 8.7498>, rot=<-0.0412 -0.1499 -0.0814 -0.9845>}
 FACIAL_R_EyelidUpperFurrow {parent=FACIAL_C_FacialRoot, loc=<-3.2079 -1.7938 8.7477>, rot=<-0.0412 0.1499 0.0813 -0.9845>}
 FACIAL_L_EyeCornerInner {parent=FACIAL_C_FacialRoot, loc=<1.5603 -0.708 7.9047>, rot=<-0.0113 0 0 -0.9999>}
 FACIAL_R_EyeCornerInner {parent=FACIAL_C_FacialRoot, loc=<-1.5639 -0.6992 7.9025>, rot=<-0.0114 0 0 -0.9999>}
 FACIAL_L_EyeCornerOuter {parent=FACIAL_C_FacialRoot, loc=<4.5473 -0.9386 7.4244>, rot=<-0.0105 -0.3826 -0.0043 -0.9239>}
 FACIAL_R_EyeCornerOuter {parent=FACIAL_C_FacialRoot, loc=<-4.5563 -0.9337 7.416>, rot=<-0.0105 0.3826 0.0043 -0.9239>}
 FACIAL_L_CheekOuter {parent=FACIAL_C_FacialRoot, loc=<4.0269 1.2978 5.8356>, rot=<0.0448 -0.378 0.0961 -0.9197>}
 FACIAL_R_CheekOuter {parent=FACIAL_C_FacialRoot, loc=<-4.0213 1.3046 5.8392>, rot=<0.0447 0.3781 -0.0961 -0.9197>}
 FACIAL_L_NasolabialBulge {parent=FACIAL_C_FacialRoot, loc=<3.4168 4.3735 5.8393>, rot=<0.0474 -0.3776 0.0195 -0.9246>}
 FACIAL_R_NasolabialBulge {parent=FACIAL_C_FacialRoot, loc=<-3.4125 4.3753 5.8356>, rot=<0.0474 0.3776 -0.0195 -0.9245>}
 FACIAL_L_NasolabialFurrow {parent=FACIAL_C_FacialRoot, loc=<2.8636 3.2763 8.2877>, rot=<-0.0437 -0.1217 -0.365 -0.922>}
 FACIAL_R_NasolabialFurrow {parent=FACIAL_C_FacialRoot, loc=<-2.8635 3.2755 8.2864>, rot=<-0.0437 0.1217 0.365 -0.922>}
 FACIAL_L_CheekLower {parent=FACIAL_C_FacialRoot, loc=<3.848 4.0861 5.1227>, rot=<0.0244 -0.4275 0.0529 -0.9021>}
 FACIAL_R_CheekLower {parent=FACIAL_C_FacialRoot, loc=<-3.8445 4.0871 5.1193>, rot=<0.0244 0.4275 -0.0529 -0.9021>}
 FACIAL_C_Nose {parent=FACIAL_C_FacialRoot, loc=<-0.0007 0.6509 9.2791>, rot=<-0.1416 0 0 -0.9899>}
 FACIAL_C_NoseLower {parent=FACIAL_C_Nose, loc=<0.0001 2.7087 -0.4639>, rot=<0.1305 0 0 -0.9914>}
 FACIAL_L_Nostril {parent=FACIAL_C_Nose, loc=<0.8777 1.9418 -0.487>, rot=<0.0173 0.1305 0.0023 -0.9913>}
 FACIAL_R_Nostril {parent=FACIAL_C_Nose, loc=<-0.872 1.94 -0.4922>, rot=<0.0173 -0.1305 -0.0023 -0.9913>}
 FACIAL_C_NoseTip {parent=FACIAL_C_Nose, loc=<0 2.1152 0.4063>, rot=<0.0872 0 0 -0.9962>}
 FACIAL_C_MouthUpper {parent=FACIAL_C_FacialRoot, loc=<0.0022 4.6023 6.8911>, rot=<-0.0113 0 0 -0.9999>}
 FACIAL_C_LipUpper {parent=FACIAL_C_MouthUpper, loc=<-0.0037 -0.0982 2.2417>, rot=<-0.0872 0 0 -0.9962>}
 FACIAL_L_LipUpper {parent=FACIAL_C_MouthUpper, loc=<1.0899 -0.0398 2.0264>, rot=<-0.0778 -0.1601 -0.0379 -0.9833>}
 FACIAL_R_LipUpper {parent=FACIAL_C_MouthUpper, loc=<-1.0972 -0.0408 2.0234>, rot=<-0.0778 0.1601 0.0379 -0.9833>}
 FACIAL_L_LipUpperOuter {parent=FACIAL_C_MouthUpper, loc=<1.8186 0.0807 1.5008>, rot=<-0.0984 -0.2293 -0.0983 -0.9634>}
 FACIAL_R_LipUpperOuter {parent=FACIAL_C_MouthUpper, loc=<-1.8275 0.0807 1.4968>, rot=<-0.0984 0.2293 0.0983 -0.9634>}
 FACIAL_L_Masseter {parent=FACIAL_C_FacialRoot, loc=<6.2253 2.0467 3.5645>, rot=<0.0581 0.1128 0.0092 -0.9919>}
 FACIAL_R_Masseter {parent=FACIAL_C_FacialRoot, loc=<-6.2267 2.0472 3.5634>, rot=<0.0581 -0.1128 -0.0092 -0.9919>}
 FACIAL_L_UnderChin {parent=FACIAL_C_FacialRoot, loc=<2.8853 7.7216 2.6155>, rot=<0.393 -0.3558 0.0783 -0.8443>}
 FACIAL_R_UnderChin {parent=FACIAL_C_FacialRoot, loc=<-2.8859 7.7222 2.6158>, rot=<0.393 0.3558 -0.0783 -0.8443>}
 FACIAL_C_LowerLipRotation {parent=FACIAL_C_FacialRoot, loc=<-0.0004 2.0743 1.8319>, rot=<0.3314 0 0 -0.9435>}
 FACIAL_C_MouthLower {parent=FACIAL_C_LowerLipRotation, loc=<0.0027 -1.1908 5.5289>, rot=<-0.342 0 0 -0.9397>}
 FACIAL_C_LipLower {parent=FACIAL_C_MouthLower, loc=<-0.0008 1.2817 1.9311>, rot=<0.0872 0 0 -0.9962>}
 FACIAL_L_LipLower {parent=FACIAL_C_MouthLower, loc=<0.953 1.2027 1.7725>, rot=<0.0927 -0.1294 0.0272 -0.9869>}
 FACIAL_R_LipLower {parent=FACIAL_C_MouthLower, loc=<-0.956 1.2023 1.7707>, rot=<0.0927 0.1294 -0.0272 -0.9869>}
 FACIAL_L_LipLowerOuter {parent=FACIAL_C_MouthLower, loc=<1.7738 0.9315 1.3427>, rot=<0.0913 -0.2321 0.0318 -0.9679>}
 FACIAL_R_LipLowerOuter {parent=FACIAL_C_MouthLower, loc=<-1.7754 0.9313 1.3389>, rot=<0.0913 0.2321 -0.0318 -0.9679>}
 FACIAL_C_Jaw {parent=FACIAL_C_FacialRoot, loc=<-0.0004 2.0743 1.8319>, rot=<0.3314 0 0 -0.9435>}
 FACIAL_C_Jawline {parent=FACIAL_C_Jaw, loc=<0.0006 2.5787 7.3109>, rot=<0.2586 0 0 -0.966>}
 FACIAL_L_ChinSide {parent=FACIAL_C_Jaw, loc=<2.5764 0.5367 7.1041>, rot=<-0.2208 -0.3738 0.0499 -0.8995>}
 FACIAL_R_ChinSide {parent=FACIAL_C_Jaw, loc=<-2.5728 0.5372 7.1045>, rot=<-0.2208 0.3738 -0.0499 -0.8995>}
 FACIAL_C_Chin {parent=FACIAL_C_Jaw, loc=<0.0012 1.1041 7.5881>, rot=<-0.2591 0 0 -0.9659>}
 FACIAL_L_Ear {parent=FACIAL_C_FacialRoot, loc=<6.8321 0.2824 0.2949>, rot=<-0.0376 -0.0004 -0.0262 -0.9989>}
 FACIAL_R_Ear {parent=FACIAL_C_FacialRoot, loc=<-6.8356 0.2805 0.3003>, rot=<-0.0376 0.0004 0.0262 -0.9989>}
 FACIAL_L_Jawline {parent=FACIAL_C_FacialRoot, loc=<4.1878 5.8096 4.1194>, rot=<0.1984 0.1584 -0.0529 -0.9658>}
 FACIAL_R_Jawline {parent=FACIAL_C_FacialRoot, loc=<-4.1883 5.8103 4.1159>, rot=<0.1984 -0.1583 0.0529 -0.9658>}
 FACIAL_C_NoseBridge {parent=FACIAL_C_FacialRoot, loc=<-0.0007 -1.3174 9.6279>, rot=<0.0325 0 0 -0.9995>}
 attach_backpack {parent=spine_05, loc=<0 -8.35 0>, rot=<-0.0093 -0.707 0.0093 -0.707>}
 clavicle_l {parent=spine_05, loc=<5.4831 -0.1053 -3.2946>, rot=<0.0465 -0.7289 -0.009 -0.6829>}
 upperarm_l {parent=clavicle_l, loc=<9.8658 -1.2136 -0.4105>, rot=<0.0164 -0.3245 -0.0269 -0.9454>}
 lowerarm_l {parent=upperarm_l, loc=<26.465 -0.8345 0.259>, rot=<0.0036 0.0161 -0.2298 -0.9731>}
 upperarm_twist_01_l {parent=upperarm_l, loc=<7.2135 -0.1134 0>, rot=<0 0 0 1>}
 deform_upperarm_bk_base_l {parent=upperarm_twist_01_l, loc=<-6.4053 -4.3746 0.314>, rot=<0 0 0 -1>}
 deform_upperarm_bk_l {parent=deform_upperarm_bk_base_l, loc=<0 0 0>, rot=<0 0 0 1>}
 clavicle_r {parent=spine_05, loc=<5.4831 -0.1053 3.2946>, rot=<0.7289 0.0465 -0.6829 0.009>}
 upperarm_r {parent=clavicle_r, loc=<-9.8658 1.2136 0.4105>, rot=<0.0164 -0.3245 -0.0269 -0.9454>}
 lowerarm_r {parent=upperarm_r, loc=<-26.465 0.8345 -0.259>, rot=<0.0036 0.0161 -0.2298 -0.9731>}
 upperarm_twist_01_r {parent=upperarm_r, loc=<-7.2135 0.1134 0>, rot=<0 0 0 -1>}
 deform_upperarm_bk_base_r {parent=upperarm_twist_01_r, loc=<6.4053 4.3746 -0.314>, rot=<0 0 0 -1>}
 deform_upperarm_bk_r {parent=deform_upperarm_bk_base_r, loc=<0 0 0>, rot=<0 0 0 1>}
 zipper_base {parent=spine_05, loc=<9.9569 6.1537 0>, rot=<0.0401 -0.9992 0 0>}
 dyn_zipper {parent=zipper_base, loc=<0 0 0>, rot=<0 0 0 1>}
 deform_pec_aim_base_l {parent=spine_04, loc=<3.2456 9.3309 -1.2237>, rot=<0.0603 -0.4033 0.3318 -0.8507>}
 deform_pec_aim_l {parent=deform_pec_aim_base_l, loc=<0 0 0>, rot=<0 0 0 1>}
 deform_pec_aim_base_r {parent=spine_04, loc=<3.2452 9.3309 1.2237>, rot=<-0.4033 -0.0603 0.8507 0.3318>}
 deform_pec_aim_r {parent=deform_pec_aim_base_r, loc=<0 0 0>, rot=<0 0 0 1>}
 thigh_l {parent=pelvis, loc=<-1.0855 -0.2109 -8.7053>, rot=<0.0009 0.0355 0.0155 -0.9993>}
 calf_l {parent=thigh_l, loc=<-42.5284 -0.0019 -0.7297>, rot=<0.0006 -0.0161 -0.0391 -0.9991>}
 foot_l {parent=calf_l, loc=<-46.7401 -2.5014 -0.592>, rot=<0.0427 -0.0238 0.0353 -0.9982>}
 calf_twist_01_l {parent=calf_l, loc=<-35.8483 -1.9197 -0.5794>, rot=<0.0237 0.0095 -0.0035 -0.9997>}
 deform_knee_in_base_l {parent=calf_l, loc=<0.8031 -5.0231 0.2456>, rot=<0 0 0 -1>}
 deform_knee_in_l {parent=deform_knee_in_base_l, loc=<0 0 0>, rot=<0 0 0 1>}
 thigh_twist_01_l {parent=thigh_l, loc=<-27.952 0 0>, rot=<0.0013 0 0 -1>}
 thigh_r {parent=pelvis, loc=<-1.0855 -0.2109 8.7053>, rot=<0.0355 -0.0009 0.9993 0.0155>}
 calf_r {parent=thigh_r, loc=<42.5284 0.0019 0.7297>, rot=<0.0006 -0.0161 -0.0391 -0.9991>}
 foot_r {parent=calf_r, loc=<46.7401 2.5014 0.592>, rot=<0.0427 -0.0238 0.0353 -0.9982>}
 calf_twist_01_r {parent=calf_r, loc=<35.8483 1.9197 0.5794>, rot=<0.0237 0.0095 -0.0035 -0.9997>}
 deform_knee_in_base_r {parent=calf_r, loc=<-0.8031 5.0231 -0.2457>, rot=<0 0 0 -1>}
 deform_knee_in_r {parent=deform_knee_in_base_r, loc=<0 0 0>, rot=<0 0 0 1>}
 thigh_twist_01_r {parent=thigh_r, loc=<27.952 0 0>, rot=<0.0013 0 0 -1>}
 deform_glute_base_l {parent=pelvis, loc=<-7.3466 -9.0213 -9.1245>, rot=<0.0172 0.031 -0.4477 -0.8935>}
 deform_glute_l {parent=deform_glute_base_l, loc=<0 0 0>, rot=<0 0 0 1>}
 deform_glute_base_r {parent=pelvis, loc=<-7.3464 -9.0213 9.1245>, rot=<0.031 -0.0172 0.8935 -0.4477>}
 deform_glute_r {parent=deform_glute_base_r, loc=<0 0 0>, rot=<0 0 0 1>}
 alt_hair_corrective {parent=pelvis, loc=<0 0 0>, rot=<-0.0127 -0.707 0.0127 -0.707>}
]
