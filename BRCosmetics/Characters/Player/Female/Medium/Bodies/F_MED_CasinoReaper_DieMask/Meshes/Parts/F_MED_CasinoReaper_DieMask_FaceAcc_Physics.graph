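# Physics asset for the face accessory of the female Casino Reaper DieMask cosmetic, setting up physics for its dyn_mask bones.
# Six small sphere bodies, three per side and offset from their bone origins, accompany a capsule on the head.
# Each mask bone is constrained directly to the head rather than chained, with identical angular limits that allow mostly one-axis swing, little cross swing and almost no twist.
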PhysicsAsset /BRCosmetics/Characters/Player/Female/Medium/Bodies/F_MED_CasinoReaper_DieMask/Meshes/Parts/F_MED_CasinoReaper_DieMask_FaceAcc_Physics
bodies {
 dyn_mask_01_l [{shapeType=Sphere, radius=1.5, shapeName=dyn_mask_01_l_sphere, center=<12.1413 0.4739 0.7031>}]
 dyn_mask_01_r [{shapeType=Sphere, radius=1.5, shapeName=dyn_mask_01_r_sphere, center=<12.2028 -0.2732 0.5157>}]
 dyn_mask_02_l [{shapeType=Sphere, radius=1.5, shapeName=dyn_mask_02_l_sphere, center=<11.7731 0.8902 0.0119>}]
 dyn_mask_02_r [{shapeType=Sphere, radius=1.5, shapeName=dyn_mask_02_r_sphere, center=<11.2372 0.2057 0.1367>}]
 dyn_mask_03_l [{shapeType=Sphere, radius=1.5, shapeName=dyn_mask_03_l_sphere, center=<10.141 0.4114 -0.1798>}]
 dyn_mask_03_r [{shapeType=Sphere, radius=1.5, shapeName=dyn_mask_03_r_sphere, center=<8.6926 -0.3999 0.1782>}]
 head [{shapeType=Capsule, length=3.5, radius=5.5, center=<-0.2508 -0.5374 -0.1236>, rotation=<rot 90 180 174.623>}]
}
constraints [
 {bone1Name=dyn_mask_01_l, bone2Name=head, swing1LimitDegrees=5, swing2LimitDegrees=30, twistLimitDegrees=1}
 {bone1Name=dyn_mask_01_r, bone2Name=head, swing1LimitDegrees=5, swing2LimitDegrees=30, twistLimitDegrees=1}
 {bone1Name=dyn_mask_02_l, bone2Name=head, swing1LimitDegrees=5, swing2LimitDegrees=30, twistLimitDegrees=1}
 {bone1Name=dyn_mask_02_r, bone2Name=head, swing1LimitDegrees=5, swing2LimitDegrees=30, twistLimitDegrees=1}
 {bone1Name=dyn_mask_03_l, bone2Name=head, swing1LimitDegrees=5, swing2LimitDegrees=30, twistLimitDegrees=1}
 {bone1Name=dyn_mask_03_r, bone2Name=head, swing1LimitDegrees=5, swing2LimitDegrees=30, twistLimitDegrees=1}
]
