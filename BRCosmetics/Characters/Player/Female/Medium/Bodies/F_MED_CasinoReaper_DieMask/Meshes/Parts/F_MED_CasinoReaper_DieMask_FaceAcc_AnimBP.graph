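# Post-copy AnimBP for the CasinoReaper DieMask face accessory, animating its hat and hair tails with procedural correctives and physics.
# Copied parent and body poses get leg and head-rotation correctives, hat AnimDynamics and hair constraints, then are cached.
# Backbling type selects default, backpack or cape RigidBody simulation from that cache, before final hat tilt.
# The event graph computes corrective alphas and drives state-based cloth, wind and gravity through deprecated helpers.
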AnimBlueprint /BRCosmetics/Characters/Player/Female/Medium/Bodies/F_MED_CasinoReaper_DieMask/Meshes/Parts/F_MED_CasinoReaper_DieMask_FaceAcc_AnimBP
target_skeleton=&/F_MED_CasinoReaper_DieMask_FaceAcc_Skeleton
graphs [
 AnimGraph {nodes [
   Control Rig() {
    alphaBoolBlend {blendOption=Linear}
    controlRigAssetReference {blueprintRigClass=&/F_MED_CasinoReaper_DieMask_FaceAcc_CtrlRig.F_MED_CasinoReaper_DieMask_FaceAcc_CtrlRig_C}
   }
   n1 = "Constraint - dyn_hair_tail_02_l (P)"(ComponentPose=Local To Component(LocalPose=Use cached pose 'pre sim'()), Alpha=0.8, ConstraintWeights_0=1) {
    alphaBoolBlend {blendOption=Linear}
    boneToModify {boneName=dyn_hair_tail_02_l}
    constraintSetup [{targetBone {boneName=spine_05}}]
    constraintWeights [1]
   }
   n2 = "Constraint - dyn_hair_tail_02_r (P)"(ComponentPose=@n1, Alpha=0.8, ConstraintWeights_0=1) {
    alphaBoolBlend {blendOption=Linear}
    boneToModify {boneName=dyn_hair_tail_02_r}
    constraintSetup [{targetBone {boneName=spine_05}}]
    constraintWeights [1]
   }
   n3 = "Constraint - dyn_hair_tail_03_l (PP)"!(ComponentPose=@n2, Alpha=0.8, ConstraintWeights_0=0.5, ConstraintWeights_1:float=0.5) -> (Pose) {
    alphaBoolBlend {blendOption=Linear}
    boneToModify {boneName=dyn_hair_tail_03_l}
    constraintSetup [{targetBone {boneName=spine_04}}, {targetBone {boneName=spine_03}}]
    constraintWeights [1,1]
   }
   n4 = "Constraint - dyn_hair_tail_03_r (PP)"!(ComponentPose=@n3, Alpha=0.8, ConstraintWeights_0=0.5, ConstraintWeights_1:float=0.5) -> (Pose) {
    alphaBoolBlend {blendOption=Linear}
    boneToModify {boneName=dyn_hair_tail_03_r}
    constraintSetup [{targetBone {boneName=spine_04}}, {targetBone {boneName=spine_03}}]
    constraintWeights [1,1]
   }
   n5 = "Constraint - dyn_hair_tail_04_l (P)"(ComponentPose=@n4, Alpha=0.8, ConstraintWeights_0=1) {
    alphaBoolBlend {blendOption=Linear}
    boneToModify {boneName=dyn_hair_tail_04_l}
    constraintSetup [{targetBone {boneName=spine_03}}]
    constraintWeights [1]
   }
   n6 = "Constraint - dyn_hair_tail_04_l (P)"(ComponentPose=@n5, Alpha=0.8, ConstraintWeights_0=1) {
    alphaBoolBlend {blendOption=Linear}
    boneToModify {boneName=dyn_hair_tail_04_l}
    constraintSetup [{targetBone {boneName=spine_03}}]
    constraintWeights [1]
   }
   n7 = "Transform (Modify) Bone - Bone: dyn_hair_tail_02_l"!(ComponentPose=@n6, bAlphaBoolEnabled:bool=True, Translation="0,0,0", Rotation="0,0,-4", Scale="1,1,1") -> (Pose) {
    alphaInputType=Bool
    rotationMode=BMM_Additive
    translationMode=BMM_Additive
    alphaBoolBlend {blendOption=Linear}
    boneToModify {boneName=dyn_hair_tail_02_l}
    rotation {roll=-4}
   }
   n8 = "Transform (Modify) Bone - Bone: dyn_hair_tail_02_r"!(ComponentPose=@n7, bAlphaBoolEnabled:bool=True, Translation="0,0,0", Rotation="0,0,-4", Scale="1,1,1") -> (Pose) {
    alphaInputType=Bool
    rotationMode=BMM_Additive
    translationMode=BMM_Additive
    alphaBoolBlend {blendOption=Linear}
    boneToModify {boneName=dyn_hair_tail_02_r}
    rotation {roll=-4}
   }
   n9 = "Transform (Modify) Bone - Bone: dyn_hair_tail_03_l"!(ComponentPose=@n8, bAlphaBoolEnabled:bool=True, Translation="0,2,0", Rotation="0,0,0", Scale="1,1,1") -> (Pose) {
    alphaInputType=Bool
    rotationMode=BMM_Additive
    translationMode=BMM_Additive
    alphaBoolBlend {blendOption=Linear}
    boneToModify {boneName=dyn_hair_tail_03_l}
    translation {y=2}
   }
   n10 = "Transform (Modify) Bone - Bone: dyn_hair_tail_03_r"!(ComponentPose=@n9, bAlphaBoolEnabled:bool=True, Translation="0,2,0", Rotation="0,0,0", Scale="1,1,1") -> (Pose) {
    alphaInputType=Bool
    rotationMode=BMM_Additive
    translationMode=BMM_Additive
    alphaBoolBlend {blendOption=Linear}
    boneToModify {boneName=dyn_hair_tail_03_r}
    translation {y=2}
   }
   n11 = Get bIsDBNO()
   n12 = "Transform (Modify) Bone - Bone: dyn_hair_tail_03_l"(ComponentPose=@n10, bAlphaBoolEnabled=@n11, Translation="0,0,0", Rotation="0,0,-15", Scale="1,1,1") {
    alphaInputType=Bool
    rotationMode=BMM_Additive
    translationMode=BMM_Additive
    alphaBoolBlend {blendOption=Linear}
    boneToModify {boneName=dyn_hair_tail_03_l}
    rotation {roll=-15}
   }
   n13 = "Transform (Modify) Bone - Bone: dyn_hair_tail_03_r"(ComponentPose=@n12, bAlphaBoolEnabled=@n11, Translation="0,0,0", Rotation="0,3.889,-15", Scale="1,1,1") {
    alphaInputType=Bool
    rotationMode=BMM_Additive
    translationMode=BMM_Additive
    alphaBoolBlend {blendOption=Linear}
    boneToModify {boneName=dyn_hair_tail_03_r}
    rotation {roll=-15, yaw=3.889}
   }
   n14 = RigidBody!(ComponentPose=@n13, Alpha, OverrideWorldGravity:Vector="0,0,-920", ExternalForce) -> (Pose) {
    bAllowPoseScalePropagation=false
    bClampLinearTranslationLimitToRefPose=true
    bOverrideWorldGravity=true
    overridePhysicsAsset=/BRCosmetics/Characters/Player/Female/Medium/Bodies/F_MED_CasinoReaper_Die/Meshes/Parts/F_MED_CasinoReaper_Die_Backpack_FaceAcc_Physics
    simulationSpace=BaseBoneSpace
    alphaBoolBlend {blendOption=Linear}
    baseBoneRef {boneName=spine_01}
    componentAppliedLinearAccClamp=<3000 3000 3000>
    overrideWorldGravity {z=-920}
    simSpaceSettings {maxAngularAcceleration=40, maxAngularVelocity=3, maxLinearAcceleration=40, maxLinearVelocity=50, velocityScaleZ=0, worldAlpha=1}
   }
   n15 = RigidBody(ComponentPose=Local To Component(LocalPose=Use cached pose 'pre sim'()), OverrideWorldGravity="0,0,-920") {
    bAllowPoseScalePropagation=false
    bClampLinearTranslationLimitToRefPose=true
    bOverrideWorldGravity=true
    overridePhysicsAsset=/BRCosmetics/Characters/Player/Female/Medium/Bodies/F_MED_CasinoReaper_Die/Meshes/Parts/F_MED_CasinoReaper_Die_FaceAcc_Physics
    simulationSpace=BaseBoneSpace
    alphaBoolBlend {blendOption=Linear}
    baseBoneRef {boneName=spine_01}
    componentAppliedLinearAccClamp=<3000 3000 3000>
    overrideWorldGravity {z=-920}
    simSpaceSettings {maxAngularAcceleration=40, maxAngularVelocity=3, maxLinearAcceleration=40, maxLinearVelocity=50, velocityScaleZ=0.5, worldAlpha=1}
   }
   n16 = Blend Poses by bool(
    BlendPose_0=Component To Local(ComponentPose=@n14)
    BlendPose_1=Component To Local(ComponentPose=@n15)
    bActiveValue=Get IsWearingBackbling() -> (IsWearingBackbling:bool)
   ) {blendType=Linear}
   n17 = "Constraint - dyn_hair_tail_02_l (P)"(ComponentPose=Local To Component(LocalPose=Use cached pose 'pre sim'()), Alpha=1, ConstraintWeights_0=1) {
    alphaBoolBlend {blendOption=Linear}
    boneToModify {boneName=dyn_hair_tail_02_l}
    constraintSetup [{targetBone {boneName=spine_05}}]
    constraintWeights [1]
   }
   n18 = "Constraint - dyn_hair_tail_02_r (P)"(ComponentPose=@n17, Alpha=1, ConstraintWeights_0=1) {
    alphaBoolBlend {blendOption=Linear}
    boneToModify {boneName=dyn_hair_tail_02_r}
    constraintSetup [{targetBone {boneName=spine_05}}]
    constraintWeights [1]
   }
   n19 = "Constraint - dyn_hair_tail_03_l (PP)"(ComponentPose=@n18, Alpha=0.8, ConstraintWeights_0=0.5, ConstraintWeights_1=0.5) {
    alphaBoolBlend {blendOption=Linear}
    boneToModify {boneName=dyn_hair_tail_03_l}
    constraintSetup [{targetBone {boneName=spine_04}}, {targetBone {boneName=spine_03}}]
    constraintWeights [1,1]
   }
   n20 = "Constraint - dyn_hair_tail_03_r (PP)"(ComponentPose=@n19, Alpha=0.8, ConstraintWeights_0=0.5, ConstraintWeights_1=0.5) {
    alphaBoolBlend {blendOption=Linear}
    boneToModify {boneName=dyn_hair_tail_03_r}
    constraintSetup [{targetBone {boneName=spine_04}}, {targetBone {boneName=spine_03}}]
    constraintWeights [1,1]
   }
   n21 = "Constraint - dyn_hair_tail_04_l (P)"(ComponentPose=@n20, Alpha=0.5, ConstraintWeights_0=1) {
    alphaBoolBlend {blendOption=Linear}
    boneToModify {boneName=dyn_hair_tail_04_l}
    constraintSetup [{targetBone {boneName=spine_03}}]
    constraintWeights [1]
   }
   n22 = "Constraint - dyn_hair_tail_04_l (P)"(ComponentPose=@n21, Alpha=0.5, ConstraintWeights_0=1) {
    alphaBoolBlend {blendOption=Linear}
    boneToModify {boneName=dyn_hair_tail_04_l}
    constraintSetup [{targetBone {boneName=spine_03}}]
    constraintWeights [1]
   }
   n23 = Get bIsDBNO()
   n24 = "Transform (Modify) Bone - Bone: dyn_hair_tail_03_l"(ComponentPose=@n22, bAlphaBoolEnabled=@n23, Translation="0,0,0", Rotation="0,0,-15", Scale="1,1,1") {
    alphaInputType=Bool
    rotationMode=BMM_Additive
    translationMode=BMM_Additive
    alphaBoolBlend {blendOption=Linear}
    boneToModify {boneName=dyn_hair_tail_03_l}
    rotation {roll=-15}
   }
   n25 = "Transform (Modify) Bone - Bone: dyn_hair_tail_03_r"(ComponentPose=@n24, bAlphaBoolEnabled=@n23, Translation="0,0,0", Rotation="0,3.889,-15", Scale="1,1,1") {
    alphaInputType=Bool
    rotationMode=BMM_Additive
    translationMode=BMM_Additive
    alphaBoolBlend {blendOption=Linear}
    boneToModify {boneName=dyn_hair_tail_03_r}
    rotation {roll=-15, yaw=3.889}
   }
   n26 = RigidBody(ComponentPose=@n25, OverrideWorldGravity="0,0,-920") {
    bAllowPoseScalePropagation=false
    bClampLinearTranslationLimitToRefPose=true
    bOverrideWorldGravity=true
    overridePhysicsAsset=/BRCosmetics/Characters/Player/Female/Medium/Bodies/F_MED_CasinoReaper_Die/Meshes/Parts/F_MED_CasinoReaper_Die_Cape_FaceAcc_Physics
    simulationSpace=BaseBoneSpace
    alphaBoolBlend {blendOption=Linear}
    baseBoneRef {boneName=head}
    componentAppliedLinearAccClamp=<3000 3000 3000>
    overrideWorldGravity {z=-920}
    simSpaceSettings {maxAngularAcceleration=40, maxAngularVelocity=3, maxLinearAcceleration=40, maxLinearVelocity=50, velocityScaleZ=0, worldAlpha=1}
   }
   n27 = Component To Local(ComponentPose=@n26)
   n28 = Get Is Backbling A Cape() -> (Is Backbling A Cape:bool)
   n29 = Get Head_Rot_Bk() -> (Head_Rot_Bk:double)
   n30 = "Transform (Modify) Bone - Bone: dyn_hat_b_base"(
    ComponentPose=Local To Component(LocalPose=Blend Poses by bool(BlendPose_0=@n27, BlendPose_1=@n16, bActiveValue=@n28) {blendType=Linear})
    Alpha=@n29
    Translation="0,0,0"
    Rotation="0,0,10"
    Scale="1,1,1"
   ) {
    rotationMode=BMM_Additive
    translationMode=BMM_Additive
    alphaBoolBlend {blendOption=Linear}
    boneToModify {boneName=dyn_hat_b_base}
    rotation {roll=10}
   }
   n31 = "Transform (Modify) Bone - Bone: dyn_hat_lb_base"(ComponentPose=@n30, Alpha=@n29, Translation="0,0,0", Rotation="0,0,10", Scale="1,1,1") {
    rotationMode=BMM_Additive
    translationMode=BMM_Additive
    alphaBoolBlend {blendOption=Linear}
    boneToModify {boneName=dyn_hat_lb_base}
    rotation {roll=10}
   }
   n32 = "Transform (Modify) Bone - Bone: dyn_hat_rb_base"(ComponentPose=@n31, Alpha=@n29, Translation="0,0,0", Rotation="0,0,20", Scale="1,1,1") {
    rotationMode=BMM_Additive
    translationMode=BMM_Additive
    alphaBoolBlend {blendOption=Linear}
    boneToModify {boneName=dyn_hat_rb_base}
    rotation {roll=20}
   }
   Output Pose(Result=Component To Local(ComponentPose=@n32)) {layerGroup=None}
   n33 = Layered blend per bone(BasePose=Copy Pose From Mesh() {bUseAttachedParent=true}, BlendPoses_0=Copy Pose From Mesh(SourceMeshComponent=Get Body_Mesh())) {layerSetup [{branchFilters [
       {boneName=upperarm_l}
       {boneName=upperarm_r}
       {boneName=lowerarm_l}
       {boneName=lowerarm_r}
       {boneName=thigh_l}
       {boneName=thigh_r}
       {boneName=calf_l}
       {boneName=calf_r}
      ]}]}
   n34 = "Transform (Modify) Bone - Bone: alt_hair_corrective"(
    ComponentPose=Local To Component(LocalPose=@n33)
    Alpha=Get Alt Legs back correction() -> (Alt Legs back correction:double)
    Translation="0,0,0"
    Rotation="0,0,50"
    Scale="1,1,1"
   ) {
    rotationMode=BMM_Additive
    translationMode=BMM_Additive
    alphaBoolBlend {blendOption=Linear}
    alphaScaleBiasClamp {bInterpResult=true}
    boneToModify {boneName=alt_hair_corrective}
    rotation {roll=50}
   }
   n35 = RigidBody!(ComponentPose=@n34, bAlphaBoolEnabled:bool=Get Is Cloth Off() -> (Is Cloth Off:bool), ExternalForce) -> (Pose) {
    alphaInputType=Bool
    overridePhysicsAsset=&/F_MED_CasinoReaper_DieMask_FaceAcc_Physics
    simulationSpace=BaseBoneSpace
    alphaBoolBlend {blendOption=Linear}
    baseBoneRef {boneName=spine_05}
    simSpaceSettings {maxAngularAcceleration=40, maxAngularVelocity=3, maxLinearAcceleration=40, maxLinearVelocity=50, velocityScaleZ=0.5, worldAlpha=1}
   }
   n36 = Anim Dynamics - Bone: dyn_hat_f!(SimulationSpace:AnimPhysSimSpaceType Enum=Component, ComponentPose=@n35, Alpha=1) -> (Pose) {
    angularDampingOverride=1
    angularSpringConstant=20000
    bAngularSpring=true
    bOverrideAngularDamping=true
    bUseGravityOverride=true
    gravityScale=0.35
    alphaBoolBlend {blendOption=Linear}
    boundBone {boneName=dyn_hat_f}
    componentAppliedLinearAccClamp=<4000 4000 4000>
    componentLinearAccScale=<1 1 1>
    componentLinearVelScale=<1 1 1>
    gravityOverride {z=-980}
    physicsBodyDefinitions [{
      boundBone {boneName=dyn_hat_f}
      boxExtents=<30 35 35>
      constraintSetup {angularLimitsMax=<0 15 0>, angularLimitsMin=<0 -15 0>, angularTarget=<1 0 0>}
      localJointOffset=<1 0 0>
     }]
    relativeSpaceBone {boneName=spine_03}
   }
   n37 = Anim Dynamics - Bone: dyn_hat_l!(SimulationSpace:AnimPhysSimSpaceType Enum=Component, ComponentPose=@n36, Alpha=1) -> (Pose) {
    angularDampingOverride=1
    angularSpringConstant=20000
    bAngularSpring=true
    bOverrideAngularDamping=true
    bUseGravityOverride=true
    gravityScale=0.35
    alphaBoolBlend {blendOption=Linear}
    boundBone {boneName=dyn_hat_l}
    componentAppliedLinearAccClamp=<4000 4000 4000>
    componentLinearAccScale=<1 1 1>
    componentLinearVelScale=<1 1 1>
    gravityOverride {z=-980}
    physicsBodyDefinitions [{
      boundBone {boneName=dyn_hat_l}
      boxExtents=<30 35 35>
      constraintSetup {angularLimitsMax=<0 15 0>, angularLimitsMin=<0 -15 0>, angularTarget=<1 0 0>}
      localJointOffset=<1 0 0>
     }]
    relativeSpaceBone {boneName=spine_03}
   }
   n38 = Anim Dynamics - Bone: dyn_hat_b!(SimulationSpace:AnimPhysSimSpaceType Enum=Component, ComponentPose=@n37, Alpha=1) -> (Pose) {
    angularDampingOverride=1
    angularSpringConstant=20000
    bAngularSpring=true
    bOverrideAngularDamping=true
    bUseGravityOverride=true
    gravityScale=0.35
    alphaBoolBlend {blendOption=Linear}
    boundBone {boneName=dyn_hat_b}
    componentAppliedLinearAccClamp=<4000 4000 4000>
    componentLinearAccScale=<1 1 1>
    componentLinearVelScale=<1 1 1>
    gravityOverride {z=-980}
    physicsBodyDefinitions [{
      boundBone {boneName=dyn_hat_b}
      boxExtents=<30 35 35>
      constraintSetup {angularLimitsMax=<0 10 0>, angularLimitsMin=<0 -15 0>, angularTarget=<1 0 0>}
      localJointOffset=<1 0 0>
     }]
    relativeSpaceBone {boneName=spine_03}
   }
   n39 = Anim Dynamics - Bone: dyn_hat_r!(SimulationSpace:AnimPhysSimSpaceType Enum=Component, ComponentPose=@n38, Alpha=1) -> (Pose) {
    angularDampingOverride=1
    angularSpringConstant=20000
    bAngularSpring=true
    bOverrideAngularDamping=true
    bUseGravityOverride=true
    gravityScale=0.35
    alphaBoolBlend {blendOption=Linear}
    boundBone {boneName=dyn_hat_r}
    componentAppliedLinearAccClamp=<4000 4000 4000>
    componentLinearAccScale=<1 1 1>
    componentLinearVelScale=<1 1 1>
    gravityOverride {z=-980}
    physicsBodyDefinitions [{
      boundBone {boneName=dyn_hat_r}
      boxExtents=<30 35 35>
      constraintSetup {angularLimitsMax=<0 15 0>, angularLimitsMin=<0 -15 0>, angularTarget=<1 0 0>}
      localJointOffset=<1 0 0>
     }]
    relativeSpaceBone {boneName=spine_03}
   }
   n40 = Anim Dynamics - Bone: dyn_hat_rb!(SimulationSpace:AnimPhysSimSpaceType Enum=Component, ComponentPose=@n39, Alpha=1) -> (Pose) {
    angularDampingOverride=1
    angularSpringConstant=20000
    bAngularSpring=true
    bOverrideAngularDamping=true
    bUseGravityOverride=true
    gravityScale=0.35
    alphaBoolBlend {blendOption=Linear}
    boundBone {boneName=dyn_hat_rb}
    componentAppliedLinearAccClamp=<4000 4000 4000>
    componentLinearAccScale=<1 1 1>
    componentLinearVelScale=<1 1 1>
    gravityOverride {z=-980}
    physicsBodyDefinitions [{
      boundBone {boneName=dyn_hat_rb}
      boxExtents=<30 35 35>
      constraintSetup {angularLimitsMax=<0 15 0>, angularLimitsMin=<0 -15 0>, angularTarget=<1 0 0>}
      localJointOffset=<1 0 0>
     }]
    relativeSpaceBone {boneName=spine_03}
   }
   n41 = Anim Dynamics - Bone: dyn_hat_lb!(SimulationSpace:AnimPhysSimSpaceType Enum=Component, ComponentPose=@n40, Alpha=1) -> (Pose) {
    angularDampingOverride=1
    angularSpringConstant=20000
    bAngularSpring=true
    bOverrideAngularDamping=true
    bUseGravityOverride=true
    gravityScale=0.35
    alphaBoolBlend {blendOption=Linear}
    boundBone {boneName=dyn_hat_lb}
    componentAppliedLinearAccClamp=<4000 4000 4000>
    componentLinearAccScale=<1 1 1>
    componentLinearVelScale=<1 1 1>
    gravityOverride {z=-980}
    physicsBodyDefinitions [{
      boundBone {boneName=dyn_hat_lb}
      boxExtents=<30 35 35>
      constraintSetup {angularLimitsMax=<0 15 0>, angularLimitsMin=<0 -15 0>, angularTarget=<1 0 0>}
      localJointOffset=<1 0 0>
     }]
    relativeSpaceBone {boneName=spine_03}
   }
   n42 = Anim Dynamics - Bone: dyn_hat_rf!(SimulationSpace:AnimPhysSimSpaceType Enum=Component, ComponentPose=@n41, Alpha=1) -> (Pose) {
    angularDampingOverride=1
    angularSpringConstant=20000
    bAngularSpring=true
    bOverrideAngularDamping=true
    bUseGravityOverride=true
    gravityScale=0.35
    alphaBoolBlend {blendOption=Linear}
    boundBone {boneName=dyn_hat_rf}
    componentAppliedLinearAccClamp=<4000 4000 4000>
    componentLinearAccScale=<1 1 1>
    componentLinearVelScale=<1 1 1>
    gravityOverride {z=-980}
    physicsBodyDefinitions [{
      boundBone {boneName=dyn_hat_rf}
      boxExtents=<30 35 35>
      constraintSetup {angularLimitsMax=<0 10 0>, angularLimitsMin=<0 -15 0>, angularTarget=<1 0 0>}
      localJointOffset=<1 0 0>
     }]
    relativeSpaceBone {boneName=spine_03}
   }
   n43 = Anim Dynamics - Bone: dyn_hat_lf!(SimulationSpace:AnimPhysSimSpaceType Enum=Component, ComponentPose=@n42, Alpha=1) -> (Pose) {
    angularDampingOverride=1
    angularSpringConstant=20000
    bAngularSpring=true
    bOverrideAngularDamping=true
    bUseGravityOverride=true
    gravityScale=0.35
    alphaBoolBlend {blendOption=Linear}
    boundBone {boneName=dyn_hat_lf}
    componentAppliedLinearAccClamp=<4000 4000 4000>
    componentLinearAccScale=<1 1 1>
    componentLinearVelScale=<1 1 1>
    gravityOverride {z=-980}
    physicsBodyDefinitions [{
      boundBone {boneName=dyn_hat_lf}
      boxExtents=<30 35 35>
      constraintSetup {angularLimitsMax=<0 15 0>, angularLimitsMin=<0 -15 0>, angularTarget=<1 0 0>}
      localJointOffset=<1 0 0>
     }]
    relativeSpaceBone {boneName=spine_03}
   }
   n44 = "Constraint - hair_base (PP)"!(ComponentPose=@n43, Alpha=1, ConstraintWeights_0=0.4, ConstraintWeights_1:float=0.6) -> (Pose) {
    alphaBoolBlend {blendOption=Linear}
    boneToModify {boneName=hair_base}
    constraintSetup [{targetBone {boneName=spine_05}}, {targetBone {boneName=head}}]
    constraintWeights [1,1]
   }
   n45 = "Constraint - dyn_hair_tail_base_r (PP)"!(ComponentPose=@n44, Alpha=1, ConstraintWeights_0=0.6, ConstraintWeights_1:float=0.4) -> (Pose) {
    alphaBoolBlend {blendOption=Linear}
    boneToModify {boneName=dyn_hair_tail_base_r}
    constraintSetup [{targetBone {boneName=spine_05}}, {targetBone {boneName=head}}]
    constraintWeights [1,1]
   }
   n46 = "Constraint - dyn_hair_tail_base_l (PP)"!(ComponentPose=@n45, Alpha=1, ConstraintWeights_0=0.6, ConstraintWeights_1:float=0.4) -> (Pose) {
    alphaBoolBlend {blendOption=Linear}
    boneToModify {boneName=dyn_hair_tail_base_l}
    constraintSetup [{targetBone {boneName=spine_05}}, {targetBone {boneName=head}}]
    constraintWeights [1,1]
   }
   n47 = Get Head_Rot_Bk()
   n48 = "Transform (Modify) Bone - Bone: hair_base"(ComponentPose=@n46, Alpha=@n47, Translation="6.630,-3.812,1.373", Rotation="0,0,20", Scale="1,1,1") {
    rotationMode=BMM_Additive
    translationMode=BMM_Additive
    translationSpace=BCS_ParentBoneSpace
    alphaBoolBlend {blendOption=Linear}
    boneToModify {boneName=hair_base}
    rotation {roll=20}
    translation=<6.63 -3.812 1.373>
   }
   n49 = "Transform (Modify) Bone - Bone: dyn_hair_tail_base_l"(ComponentPose=@n48, Alpha=@n47, Translation="0,0,2", Rotation="0,0,0", Scale="1,1,1") {
    rotationMode=BMM_Additive
    translationMode=BMM_Additive
    alphaBoolBlend {blendOption=Linear}
    boneToModify {boneName=dyn_hair_tail_base_l}
    translation {z=2}
   }
   n50 = Get Head_Rot_right() -> (Head_Rot_right:double)
   n51 = "Transform (Modify) Bone - Bone: hair_base"(ComponentPose=@n49, Alpha=@n50, Translation="5.039,-4.077,0.551", Rotation="0.000,0.000,-4.532", Scale="1,1,1") {
    rotationMode=BMM_Additive
    translationMode=BMM_Additive
    translationSpace=BCS_ParentBoneSpace
    alphaBoolBlend {blendOption=Linear}
    boneToModify {boneName=hair_base}
    rotation {roll=-4.532}
    translation=<5.039 -4.077 0.551>
   }
   n52 = "Transform (Modify) Bone - Bone: dyn_hair_tail_base_l"(ComponentPose=@n51, Alpha=@n50, Translation="-2.653,2.648,0.000", Rotation="-0.000,0.000,11.194", Scale="1,1,1") {
    rotationMode=BMM_Additive
    translationMode=BMM_Additive
    alphaBoolBlend {blendOption=Linear}
    boneToModify {boneName=dyn_hair_tail_base_l}
    rotation {roll=11.194}
    translation=<-2.653 2.648>
   }
   n53 = Get Head Rot left() -> (Head Rot left:double)
   n54 = "Transform (Modify) Bone - Bone: hair_base"(ComponentPose=@n52, Alpha=@n53, Translation="0.165,0.000,0.000", Rotation="0,0,0", Scale="1,1,1") {
    rotationMode=BMM_Additive
    translationMode=BMM_Additive
    alphaBoolBlend {blendOption=Linear}
    boneToModify {boneName=hair_base}
    translation {x=0.165}
   }
   n55 = "Transform (Modify) Bone - Bone: dyn_hair_tail_base_r"(ComponentPose=@n54, Alpha=@n53, Translation="2.000,2.385,-0.000", Rotation="-0.000,0.000,11.194", Scale="1,1,1") {
    rotationMode=BMM_Additive
    translationMode=BMM_Additive
    alphaBoolBlend {blendOption=Linear}
    boneToModify {boneName=dyn_hair_tail_base_r}
    rotation {roll=11.194}
    translation=<2 2.385>
   }
   n56 = "Transform (Modify) Bone - Bone: dyn_hair_tail_base_l"(ComponentPose=@n55, Alpha=@n53, Translation="0,0,0", Rotation="-0.000,-11.573,0.000", Scale="1,1,1") {
    rotationMode=BMM_Additive
    translationMode=BMM_Additive
    alphaBoolBlend {blendOption=Linear}
    boneToModify {boneName=dyn_hair_tail_base_l}
    rotation {yaw=-11.573}
   }
   Save cached pose 'pre sim'(Pose=Component To Local(ComponentPose=@n56)) {cacheName=pre sim}
  ]}
 EventGraph {nodes [
   Event BlueprintUpdateAnimation()
   > n1 = Sequence()
   n2 = GetOwnerSkeletalMeshComponent()
   > Set Head_Rot_Bk(Head_Rot_Bk:double=MapRangeClamped(
     Value=GetDeltaTransformFromRefPose!(self=@n2, BoneName=head, BaseName=spine_05) -> (
      ReturnValue_Location:Vector
      ReturnValue_Rotation_Roll:float
      ReturnValue_Rotation_Pitch:float
      ReturnValue_Rotation_Yaw:float
      ReturnValue_Scale:Vector
     ).ReturnValue_Rotation_Yaw
     InRangeA=-20
     InRangeB=-90
    )) -> (Output_Get:double)
   n3 = GetOwnerSkeletalMeshComponent()
   > Set Head_Rot_right(Head_Rot_right:double=MapRangeClamped(Value=GetDeltaTransformFromRefPose(self=@n3, BoneName=head, BaseName=clavicle_l).ReturnValue_Rotation_Roll, InRangeA=-10, InRangeB=-90)) -> (Output_Get:double)
   n4 = GetOwnerSkeletalMeshComponent()
   > Set Head Rot left(Head Rot left:double=MapRangeClamped(Value=GetDeltaTransformFromRefPose(self=@n4, BoneName=head, BaseName=clavicle_r).ReturnValue_Rotation_Roll, InRangeA=10, InRangeB=90)) -> (Output_Get:double)
   > Set Is Cloth Off(Is Cloth Off:bool=NOT Boolean(A=Get bClothEnabled())) -> (Output_Get:bool)
   > n5 = Branch(Condition=Get bClothEnabled())
   > n6 = Sequence()
   n7 = Gravity Override State Function DEPRECATED(
    execute=[@n6.then_0, @n5.else]
    CharacterPartAnimInstance=Self-Reference()
    GameTimeInSeconds=GetGameTimeInSeconds()
    GravityOverrideParams=Get Gravity Override Params()
   )
   > Set Gravity Override(Gravity Override=@n7)
   n8 = Cloth Controller Function DEPRECATED!(
    execute=[@n6.then_1, @n21.then]
    CharacterPartAnimInstance:Custom Character Part Anim Instance Object Reference=Self-Reference()
    ClothingAssetName:String=CasinoReaper_Clothing
    Cloth Paramaters:Cloth Params Struct Structure=Get Cloth Paramaters_0() -> (Cloth Paramaters_0:Cloth Params Struct Structure)
    Cloth Core Settings:Cloth Core Setting Struct Structure=Get Cloth Core Settings_0() -> (Cloth Core Settings_0:Cloth Core Setting Struct Structure)
    ClothUpAxis:Vector="1,0,0"
    OnGround_Moving_SpeedLimit:double=0
    OnGround_Moving_AngularVel_Max:double=0
    OnGround_Moving_FictitiousVel_Max:double=0
    LinVelScale_FwdBwd_Clamp:double=0
    LinVelScale_UpDwn_Clamp:double=0
    DeltaTime:double=Get DeltaTimeLocal() -> (DeltaTimeLocal:double)
    GameTimeInSeconds:double=GetGameTimeInSeconds()
    WindGustsParams:Wind Gust Params Struct Structure=Get Wind Gust Params() -> (Wind Gust Params:Wind Gust Params Struct Structure)
    WindGustLocalSettings:Wind Gust Local Setting Struct Structure=Get Gust Local Settings() -> (Gust Local Settings:Wind Gust Local Setting Struct Structure)
   ) -> (
    then
    CurrentStateOut:String
    Cloth Settings Out:Cloth Setting Struct Structure
    Cloth Core Settings Out:Cloth Core Setting Struct Structure
    WindGustSettings Out:Wind Gust Setting Struct Structure
    WindGustLocalSettings Out:Wind Gust Local Setting Struct Structure
    GustStrength - WindVelocityOut:Vector
    LinearVelVector-DELETE LATER:Vector
   )
   > Set Gust Local Settings(Gust Local Settings:Wind Gust Local Setting Struct Structure=@n8.WindGustLocalSettings Out) -> (Output_Get:Wind Gust Local Setting Struct Structure)
   n9 = MapRangeClamped(
    Value=GetDeltaTransformFromRefPose(self=GetOwnerSkeletalMeshComponent(), BoneName=calf_l, BaseName=pelvis).ReturnValue_Rotation_Yaw
    InRangeA=0
    InRangeB=70
   )
   n10 = MapRangeClamped(
    Value=GetDeltaTransformFromRefPose(self=GetOwnerSkeletalMeshComponent(), BoneName=calf_r, BaseName=pelvis).ReturnValue_Rotation_Yaw
    InRangeA=0
    InRangeB=70
   )
   n11 = Branch(execute=@n1.then_1, Condition="float > float"(A=@n10, B=@n9))
   > Set Alt Legs back correction(Alt Legs back correction:double=@n10) -> (Output_Get:double)
   Set Alt Legs back correction(execute=@n11.else, Alt Legs back correction=@n9)
   n12 = Event BlueprintInitializeAnimation()
   n13 = Sequence(execute=[@n12.then, @n20.then])
   > Set Body_Mesh(Body_Mesh=GetOwnerSkeletalMeshComponent())
   n14 = TryGetPawnOwner()
   > Is Valid(InputObject=@n14)
   > n15 = Cast To FortPlayerPawnAthena(Object=@n14)
   > Set As Fort Player Pawn Athena(As Fort Player Pawn Athena:Fort Player Pawn Athena Object Reference=@n15) -> (Output_Get:Fort Player Pawn Athena Object Reference)
   > n16 = Cast To PlayerPawn_Athena(Object=@n14)
   > Set Player Pawn(Player Pawn:Player Pawn Athena Object Reference=@n16) -> (Output_Get:Player Pawn Athena Object Reference)
   n17 = Sequence(execute=@n13.then_1)
   n18 = Branch(execute=@n17.then_1, Condition=Get bIsInFrontEnd())
   > Set Sim Space C(Sim Space C:AnimPhysSimSpaceType Enum=World) -> (Output_Get:AnimPhysSimSpaceType Enum)
   Set Sim Space C(execute=@n18.else, Sim Space C=Component)
   n19 = CheckForBackbling()
   Set IsWearingBackbling(execute=@n17.then_0, IsWearingBackbling:bool=@n19) -> (Output_Get:bool)
   > Set Is Backbling A Cape(Is Backbling A Cape:bool=@n19.bIsBackblingACape) -> (Output_Get:bool)
   n20 = Event OnCharacterCustomizationLoaded()
   n21 = Branch(Condition=Get bIsSkyDiving())
   DrawVectorFromJointFunction!(
    execute
    SkelMeshComponent:Scene Component Object Reference
    JointName:Name=pelvis
    LineEnd:Vector=Break Cloth Setting Struct!(ClothSettingStruct:"Cloth Setting Struct Structure (by ref)"=@n8.Cloth Settings Out) -> (
     LinearVelocityScale_7_#1:Vector=~
     AngularVelocityScale_6_#2:float=~
     FictitiousAngularScale_10_#3:float=~
     GravityOverride_15_#4:Vector=~
     DragCoefficient_23_#5:Vector 2D Structure
     LiftCoefficient_24_#6:Vector 2D Structure
     JointName_27_#7:Name
    ) {is=title K2Node_BreakStruct}.GravityOverride_15_#4
    JointOffset:Vector="0, 0, 0"
    LineColor:Linear Color Structure=(R=1,G=0,B=0,A=1)
    Thickness:double=2
    Foreground:EDrawDebugSceneDepthPriorityGroup Enum=Foreground
   ) -> (then)
  ]}
]
variables [
 Alt Legs back correction {type=number}
 As Fort Player Pawn Athena {type=FortPlayerPawnAthena}
 Body_Mesh {type=SkeletalMeshComponent}
 Cloth Core Settings_0 {
  type=ClothCoreSettingStruct
  default {
   collisionThickness=1
   dampingCoefficient=0.2
   frictionCoefficient=0
   localDampingCoefficient=0.1
   numberIterations=2
   numberSubsteps=1
   selfCollisionThickness=0
   animDriveDamping=<0 1>
   animDriveStiffness=<0 1>
   areaStiffness=<1 1>
   bendingStiffness=<1 0.5>
   edgeStiffness=<1 0.5>
   tetherStiffness=<1 1>
  }
 }
 Cloth Paramaters_0 {
  type=ClothParamsStruct
  default {
   dBNO {
    angularVelocityScale=0.6
    fictitiousAngularScale=0.1
    jointName=pelvis
    dragCoefficient=<0.2 1>
    gravityOverride=<-980 400 0>
    liftCoefficient=<0 0.5>
    linearVelocityScale=<0.3 0.3 0.1>
   }
   dBNO_Carried {
    angularVelocityScale=0.6
    fictitiousAngularScale=0.1
    jointName=pelvis
    dragCoefficient=<0.2 1>
    gravityOverride=<-980 -200 0>
    liftCoefficient=<0 0.5>
    linearVelocityScale=<0.3 0.3 0.1>
   }
   dBNO_Impulse {
    angularVelocityScale=0
    fictitiousAngularScale=0
    jointName=pelvis
    dragCoefficient=<0 0.2>
    gravityOverride=<-3000 -2000 0>
    liftCoefficient=<0 0.1>
    linearVelocityScale=<0.3 0.1 0.3>
   }
   dBNO_RollBwd {
    angularVelocityScale=0
    fictitiousAngularScale=0
    jointName=pelvis
    dragCoefficient=<0 0.2>
    gravityOverride=<-2000 0 0>
    liftCoefficient=<0 0.1>
    linearVelocityScale=<0.3 0.1 0.3>
   }
   dBNO_RollFwd {
    angularVelocityScale=0
    fictitiousAngularScale=0
    jointName=pelvis
    dragCoefficient=<0 0.2>
    gravityOverride=<-980 0 0>
    liftCoefficient=<0 0.1>
    linearVelocityScale=<0.3 0.1 0.3>
   }
   dBNO_Tumbling {
    angularVelocityScale=0
    fictitiousAngularScale=0
    jointName=pelvis
    dragCoefficient=<0 0.2>
    gravityOverride=<-2500 -300 0>
    liftCoefficient=<0 0.1>
    linearVelocityScale=<0.3 0.1 0.3>
   }
   emoteOrMelee {
    angularVelocityScale=0.6
    fictitiousAngularScale=0.1
    jointName=root
    dragCoefficient=<2.5 2.5>
    gravityOverride=<0 0 -980>
    liftCoefficient=<0.5 0.5>
    linearVelocityScale=<0.6 0.6 0.5>
   }
   front_End {
    angularVelocityScale=0.05
    fictitiousAngularScale=0.6
    jointName=root
    dragCoefficient=<0.2 0.2>
    gravityOverride=<0 -200 -980>
    liftCoefficient=<0.1 0.1>
    linearVelocityScale=<0.1 0.1 0.1>
   }
   onGround_Crouch_Moving {
    angularVelocityScale=0.1
    fictitiousAngularScale=0.3
    jointName=root
    dragCoefficient=<2 2>
    gravityOverride=<0 0 -980>
    liftCoefficient=<0.5 0.5>
    linearVelocityScale=<0.6 0.6 0.5>
   }
   onGround_Falling {
    angularVelocityScale=0.6
    fictitiousAngularScale=0
    jointName=root
    dragCoefficient=<0 0.2>
    gravityOverride=<0 0 -980>
    liftCoefficient=<0 0.1>
    linearVelocityScale=<0.1 0.1 0.1>
   }
   onGround_Moving {
    angularVelocityScale=0.6
    fictitiousAngularScale=0.1
    jointName=root
    dragCoefficient=<2 2>
    gravityOverride=<0 0 -980>
    liftCoefficient=<0.5 0.5>
    linearVelocityScale=<0.1 0.1 0.1>
   }
   onGround_Standing {
    angularVelocityScale=0.05
    fictitiousAngularScale=1.2
    jointName=root
    dragCoefficient=<2.5 2.5>
    gravityOverride=<0 0 -980>
    liftCoefficient=<0.5 0.5>
    linearVelocityScale=<0.1 0.1 0.1>
   }
   rail_Grinding {
    angularVelocityScale=0.2
    fictitiousAngularScale=0
    jointName=root
    dragCoefficient=<0.1 0.2>
    gravityOverride=<50 -500 -400>
    liftCoefficient=<0 0.5>
    linearVelocityScale=<0.1 0.1 0.1>
   }
   riding_Moving {
    angularVelocityScale=0.05
    fictitiousAngularScale=1
    jointName=root
    dragCoefficient=<0.5 0.5>
    gravityOverride=<0 0 -980>
    liftCoefficient=<0.5 0.5>
    linearVelocityScale=<0.4 0.6 0.3>
   }
   riding_Moving_Targeting {
    angularVelocityScale=0.05
    fictitiousAngularScale=0.7
    jointName=root
    dragCoefficient=<0.5 0.5>
    gravityOverride=<0 0 -980>
    liftCoefficient=<0.5 0.5>
    linearVelocityScale=<0.1 0.1 0.1>
   }
   riding_Standing {
    angularVelocityScale=0.05
    fictitiousAngularScale=1
    jointName=root
    dragCoefficient=<0.5 0.5>
    gravityOverride=<0 -450 -980>
    liftCoefficient=<0.5 0.5>
    linearVelocityScale=<0.6 0.6 0.3>
   }
   skydive_DivingDown {
    angularVelocityScale=0.1
    fictitiousAngularScale=0
    jointName=pelvis
    dragCoefficient=<0.3 0.3>
    gravityOverride=<-3000 -100 0>
    liftCoefficient=<0.9 0.9>
    linearVelocityScale=<0.1 0.1 0.1>
   }
   skydive_FloatingDown {
    angularVelocityScale=0.1
    fictitiousAngularScale=0
    jointName=root
    dragCoefficient=<0.3 0.3>
    gravityOverride=<0 -800 200>
    liftCoefficient=<1.3 1.3>
    linearVelocityScale=<0.2 0.2 0.1>
   }
   skydive_GoingUp {
    angularVelocityScale=0.1
    fictitiousAngularScale=0
    jointName=root
    dragCoefficient=<0.01 0.01>
    gravityOverride=<0 -2000 400>
    liftCoefficient=<0.05 0.05>
    linearVelocityScale=<0.1 0.1 0.01>
   }
   skydive_Parachuting {
    angularVelocityScale=0.6
    fictitiousAngularScale=0.1
    jointName=root
    dragCoefficient=<0.8 0.8>
    gravityOverride=<0 -600 -500>
    liftCoefficient=<0.8 0.8>
    linearVelocityScale=<0.2 0.2 0.3>
   }
   skydive_Surfing {
    angularVelocityScale=0.6
    fictitiousAngularScale=0.1
    jointName=root
    dragCoefficient=<1.5 1.5>
    gravityOverride=<0 -600 -500>
    liftCoefficient=<1 1>
    linearVelocityScale=<0.1 0.1 0.3>
   }
   swimming {
    angularVelocityScale=0.6
    fictitiousAngularScale=0
    jointName=root
    dragCoefficient=<2.5 2.5>
    gravityOverride=<0 0 0>
    liftCoefficient=<0.1 0.1>
    linearVelocityScale=<0.6 0.6 0.5>
   }
   tactical_Sprint {
    angularVelocityScale=0.6
    fictitiousAngularScale=0.1
    jointName=root
    dragCoefficient=<2 2>
    gravityOverride=<0 0 -980>
    liftCoefficient=<0.5 0.5>
    linearVelocityScale=<0.1 0.1 0.1>
   }
   ziplining {
    angularVelocityScale=0.6
    fictitiousAngularScale=0
    jointName=root
    dragCoefficient=<0 1.5>
    gravityOverride=<0 0 -980>
    liftCoefficient=<0 0.3>
    linearVelocityScale=<0.4 0.4 0.5>
   }
  }
 }
 DeltaTimeLocal {type=number}
 Gravity Override {type=Vector}
 Gravity Override Params {type=GravityOverrideParamsStruct, default {dBNO {gravityOverride=<-2000 1500 0>}}}
 GravityOverride {type=Vector}
 Gust Local Settings {type=WindGustLocalSettingStruct, default {elapsedTime=0, finalGustStrengthLocal=0, gustDuration=0, gustStrength=0, timeBetweenGusts=0}}
 Head Rot left {type=number}
 Head_Rot_Bk {type=number}
 Head_Rot_right {type=number}
 Is Backbling A Cape {type=boolean}
 Is Cloth Off {type=boolean}
 IsShortBackblingOnAlpha {type=number}
 IsWearingBackbling {type=boolean}
 Player Pawn {type=PlayerPawn_Athena_C}
 Sim Space C {type=AnimPhysSimSpaceType}
 Wind Gust Params {
  type=WindGustParamsStruct
  default {
   dBNO {
    maxDownInterpSpeed=5
    maxGustDuration=2
    maxGustStrength=7
    maxTimeBetweenGusts=7
    maxUpInterpSpeed=5
    minDownInterpSpeed=2
    minGustDuration=0.5
    minGustStrength=3
    minTimeBetweenGusts=0
    minUpInterpSpeed=2
    windTurbulenceScalarA=0.1
    windTurbulenceScalarB=1
    windAmplitude=<0 0 0>
    windFrequency=<0 0 0>
    windOffset=<0 0 0>
   }
   dBNO_Carried {
    maxDownInterpSpeed=5
    maxGustDuration=2
    maxGustStrength=7
    maxTimeBetweenGusts=7
    maxUpInterpSpeed=5
    minDownInterpSpeed=2
    minGustDuration=0.5
    minGustStrength=3
    minTimeBetweenGusts=0
    minUpInterpSpeed=2
    windTurbulenceScalarA=0.1
    windTurbulenceScalarB=1
    windAmplitude=<0 0 0>
    windFrequency=<0 0 0>
    windOffset=<0 0 0>
   }
   dBNO_Impulse {
    maxDownInterpSpeed=5
    maxGustDuration=2
    maxGustStrength=7
    maxTimeBetweenGusts=7
    maxUpInterpSpeed=5
    minDownInterpSpeed=2
    minGustDuration=0.5
    minGustStrength=3
    minTimeBetweenGusts=0
    minUpInterpSpeed=2
    windTurbulenceScalarA=0.1
    windTurbulenceScalarB=1
    windAmplitude=<0 0 0>
    windFrequency=<0 0 0>
    windOffset=<0 0 0>
   }
   dBNO_RollBwd {
    maxDownInterpSpeed=5
    maxGustDuration=2
    maxGustStrength=7
    maxTimeBetweenGusts=7
    maxUpInterpSpeed=5
    minDownInterpSpeed=2
    minGustDuration=0.5
    minGustStrength=3
    minTimeBetweenGusts=0
    minUpInterpSpeed=2
    windTurbulenceScalarA=0.1
    windTurbulenceScalarB=1
    windAmplitude=<0 0 0>
    windFrequency=<0 0 0>
    windOffset=<0 0 0>
   }
   dBNO_RollFwd {
    maxDownInterpSpeed=5
    maxGustDuration=2
    maxGustStrength=7
    maxTimeBetweenGusts=7
    maxUpInterpSpeed=5
    minDownInterpSpeed=2
    minGustDuration=0.5
    minGustStrength=3
    minTimeBetweenGusts=0
    minUpInterpSpeed=2
    windTurbulenceScalarA=0.1
    windTurbulenceScalarB=1
    windAmplitude=<0 0 0>
    windFrequency=<0 0 0>
    windOffset=<0 0 0>
   }
   dBNO_Tumbling {
    maxDownInterpSpeed=5
    maxGustDuration=2
    maxGustStrength=7
    maxTimeBetweenGusts=7
    maxUpInterpSpeed=5
    minDownInterpSpeed=2
    minGustDuration=0.5
    minGustStrength=3
    minTimeBetweenGusts=0
    minUpInterpSpeed=2
    windTurbulenceScalarA=0.1
    windTurbulenceScalarB=1
    windAmplitude=<0 0 0>
    windFrequency=<0 0 0>
    windOffset=<0 0 0>
   }
   emoteOrMelee {
    maxDownInterpSpeed=5
    maxGustDuration=2
    maxGustStrength=7
    maxTimeBetweenGusts=7
    maxUpInterpSpeed=5
    minDownInterpSpeed=2
    minGustDuration=0.5
    minGustStrength=3
    minTimeBetweenGusts=0
    minUpInterpSpeed=2
    windTurbulenceScalarA=0.1
    windTurbulenceScalarB=1
    windAmplitude=<0 0 0>
    windFrequency=<0 0 0>
    windOffset=<0 0 0>
   }
   front_End {
    maxDownInterpSpeed=5
    maxGustDuration=2
    maxGustStrength=7
    maxTimeBetweenGusts=7
    maxUpInterpSpeed=5
    minDownInterpSpeed=2
    minGustDuration=0.5
    minGustStrength=3
    minTimeBetweenGusts=0
    minUpInterpSpeed=2
    windTurbulenceScalarA=0.1
    windTurbulenceScalarB=1
    windAmplitude=<0 0 0>
    windFrequency=<0 0 0>
    windOffset=<0 0 0>
   }
   onGround_Crouch_Moving {
    maxDownInterpSpeed=5
    maxGustDuration=2
    maxGustStrength=7
    maxTimeBetweenGusts=7
    maxUpInterpSpeed=5
    minDownInterpSpeed=2
    minGustDuration=0.5
    minGustStrength=3
    minTimeBetweenGusts=0
    minUpInterpSpeed=2
    windTurbulenceScalarA=0.1
    windTurbulenceScalarB=1
    windAmplitude=<0 0 0>
    windFrequency=<0 0 0>
    windOffset=<0 0 0>
   }
   onGround_Falling {
    maxDownInterpSpeed=5
    maxGustDuration=2
    maxGustStrength=7
    maxTimeBetweenGusts=7
    maxUpInterpSpeed=5
    minDownInterpSpeed=2
    minGustDuration=0.5
    minGustStrength=3
    minTimeBetweenGusts=0
    minUpInterpSpeed=2
    windTurbulenceScalarA=0.1
    windTurbulenceScalarB=1
    windAmplitude=<0 0 0>
    windFrequency=<0 0 0>
    windOffset=<0 0 0>
   }
   onGround_Moving {
    maxDownInterpSpeed=5
    maxGustDuration=2
    maxGustStrength=7
    maxTimeBetweenGusts=7
    maxUpInterpSpeed=5
    minDownInterpSpeed=2
    minGustDuration=0.5
    minGustStrength=3
    minTimeBetweenGusts=0
    minUpInterpSpeed=2
    windTurbulenceScalarA=0.1
    windTurbulenceScalarB=1
    windAmplitude=<0 0 0>
    windFrequency=<0 0 0>
    windOffset=<0 0 0>
   }
   onGround_Standing {
    maxDownInterpSpeed=5
    maxGustDuration=2
    maxGustStrength=7
    maxTimeBetweenGusts=7
    maxUpInterpSpeed=5
    minDownInterpSpeed=2
    minGustDuration=0.5
    minGustStrength=3
    minTimeBetweenGusts=0
    minUpInterpSpeed=2
    windTurbulenceScalarA=0.1
    windTurbulenceScalarB=1
    windAmplitude=<0 0 0>
    windFrequency=<0 0 0>
    windOffset=<0 0 0>
   }
   rail_Grinding {
    maxDownInterpSpeed=5
    maxGustDuration=8
    maxGustStrength=7
    maxTimeBetweenGusts=8
    maxUpInterpSpeed=5
    minDownInterpSpeed=2
    minGustDuration=0.2
    minGustStrength=3
    minTimeBetweenGusts=1
    minUpInterpSpeed=2
    windTurbulenceScalarA=0.1
    windTurbulenceScalarB=0.8
    windAmplitude=<20 65 42>
    windFrequency=<5 6 8.6>
    windOffset=<50 -500 -400>
   }
   riding_Moving {
    maxDownInterpSpeed=5
    maxGustDuration=2
    maxGustStrength=7
    maxTimeBetweenGusts=7
    maxUpInterpSpeed=5
    minDownInterpSpeed=2
    minGustDuration=0.5
    minGustStrength=3
    minTimeBetweenGusts=0
    minUpInterpSpeed=2
    windTurbulenceScalarA=0.1
    windTurbulenceScalarB=1
    windAmplitude=<0 0 0>
    windFrequency=<0 0 0>
    windOffset=<0 0 0>
   }
   riding_Moving_Targeting {
    maxDownInterpSpeed=5
    maxGustDuration=2
    maxGustStrength=7
    maxTimeBetweenGusts=7
    maxUpInterpSpeed=5
    minDownInterpSpeed=2
    minGustDuration=0.5
    minGustStrength=3
    minTimeBetweenGusts=0
    minUpInterpSpeed=2
    windTurbulenceScalarA=0.1
    windTurbulenceScalarB=1
    windAmplitude=<0 0 0>
    windFrequency=<0 0 0>
    windOffset=<0 0 0>
   }
   riding_Standing {
    maxDownInterpSpeed=5
    maxGustDuration=2
    maxGustStrength=7
    maxTimeBetweenGusts=7
    maxUpInterpSpeed=5
    minDownInterpSpeed=2
    minGustDuration=0.5
    minGustStrength=3
    minTimeBetweenGusts=0
    minUpInterpSpeed=2
    windTurbulenceScalarA=0.1
    windTurbulenceScalarB=1
    windAmplitude=<0 0 0>
    windFrequency=<0 0 0>
    windOffset=<0 0 0>
   }
   skydive_DivingDown {
    maxDownInterpSpeed=5
    maxGustDuration=2
    maxGustStrength=7
    maxTimeBetweenGusts=7
    maxUpInterpSpeed=5
    minDownInterpSpeed=2
    minGustDuration=0.5
    minGustStrength=3
    minTimeBetweenGusts=0
    minUpInterpSpeed=2
    windTurbulenceScalarA=0.1
    windTurbulenceScalarB=1
    windAmplitude=<0 0 0>
    windFrequency=<0 0 0>
    windOffset=<0 0 0>
   }
   skydive_FloatingDown {
    maxDownInterpSpeed=5
    maxGustDuration=2
    maxGustStrength=7
    maxTimeBetweenGusts=7
    maxUpInterpSpeed=5
    minDownInterpSpeed=2
    minGustDuration=0.5
    minGustStrength=3
    minTimeBetweenGusts=0
    minUpInterpSpeed=2
    windTurbulenceScalarA=0.1
    windTurbulenceScalarB=1
    windAmplitude=<0 0 0>
    windFrequency=<0 0 0>
    windOffset=<0 0 0>
   }
   skydive_GoingUp {
    maxDownInterpSpeed=5
    maxGustDuration=2
    maxGustStrength=7
    maxTimeBetweenGusts=7
    maxUpInterpSpeed=5
    minDownInterpSpeed=2
    minGustDuration=0.5
    minGustStrength=3
    minTimeBetweenGusts=0
    minUpInterpSpeed=2
    windTurbulenceScalarA=0.1
    windTurbulenceScalarB=1
    windAmplitude=<0 0 0>
    windFrequency=<0 0 0>
    windOffset=<0 0 0>
   }
   skydive_Parachuting {
    maxDownInterpSpeed=5
    maxGustDuration=2
    maxGustStrength=7
    maxTimeBetweenGusts=7
    maxUpInterpSpeed=5
    minDownInterpSpeed=2
    minGustDuration=0.5
    minGustStrength=3
    minTimeBetweenGusts=0
    minUpInterpSpeed=2
    windTurbulenceScalarA=0.1
    windTurbulenceScalarB=1
    windAmplitude=<0 0 0>
    windFrequency=<0 0 0>
    windOffset=<0 0 0>
   }
   skydive_Surfing {
    maxDownInterpSpeed=5
    maxGustDuration=2
    maxGustStrength=7
    maxTimeBetweenGusts=7
    maxUpInterpSpeed=5
    minDownInterpSpeed=2
    minGustDuration=0.5
    minGustStrength=3
    minTimeBetweenGusts=0
    minUpInterpSpeed=2
    windTurbulenceScalarA=0.1
    windTurbulenceScalarB=1
    windAmplitude=<0 0 0>
    windFrequency=<0 0 0>
    windOffset=<0 0 0>
   }
   swimming {
    maxDownInterpSpeed=5
    maxGustDuration=2
    maxGustStrength=7
    maxTimeBetweenGusts=7
    maxUpInterpSpeed=5
    minDownInterpSpeed=2
    minGustDuration=0.5
    minGustStrength=3
    minTimeBetweenGusts=0
    minUpInterpSpeed=2
    windTurbulenceScalarA=0.1
    windTurbulenceScalarB=1
    windAmplitude=<0 0 0>
    windFrequency=<0 0 0>
    windOffset=<0 0 0>
   }
   tactical_Sprint {
    maxDownInterpSpeed=5
    maxGustDuration=2
    maxGustStrength=7
    maxTimeBetweenGusts=7
    maxUpInterpSpeed=5
    minDownInterpSpeed=2
    minGustDuration=0.5
    minGustStrength=3
    minTimeBetweenGusts=0
    minUpInterpSpeed=2
    windTurbulenceScalarA=0.1
    windTurbulenceScalarB=1
    windAmplitude=<0 0 0>
    windFrequency=<0 0 0>
    windOffset=<0 0 0>
   }
   ziplining {
    maxDownInterpSpeed=5
    maxGustDuration=2
    maxGustStrength=7
    maxTimeBetweenGusts=7
    maxUpInterpSpeed=5
    minDownInterpSpeed=2
    minGustDuration=0.5
    minGustStrength=3
    minTimeBetweenGusts=0
    minUpInterpSpeed=2
    windTurbulenceScalarA=0.1
    windTurbulenceScalarB=1
    windAmplitude=<0 0 0>
    windFrequency=<0 0 0>
    windOffset=<0 0 0>
   }
  }
 }
]
