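# Physics asset for the face accessory mesh of the female Casino Reaper DieMask character cosmetic.
# It holds a single body on the head bone, a small sphere offset from the bone origin.
# It defines no constraints, so there is no jointed setup between bodies.
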PhysicsAsset /BRCosmetics/Characters/Player/Female/Medium/Bodies/F_MED_CasinoReaper_DieMask/Meshes/Parts/F_MED_CasinoReaper_DieMask_cl_FaceAcc_Physics
bodies {head [{shapeType=Sphere, radius=5.35, shapeName=head_sphere, center=<-0.7716 4.0584 0>}]}
constraints []
